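# Niagara cosmetic effect that adds green necklace glows to the Thrive Spirit female medium player body.
# Its single emitter bursts seven particles multiplied by User.Variant, split into seven spawn groups that set sprite size and bone or socket index.
# Particles re-sample the user skeletal mesh each update to follow it, persist until the emitter deactivates, and render as camera-offset sprites plus small lights.
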NiagaraSystem /BRCosmetics/Characters/Player/Female/Medium/Bodies/F_MED_Thrive_Spirit/FX/NS_ThriveSpirit_Glows
dependencies {usedDataInterfaces [], usedDynamicInputs [], usedModules [&SystemState]}
emitters {Necklace_Glow {
  emitterUpdateScript [
   &EmitterState(Life Cycle Mode=::NewEnumerator1, Loop Behavior=::NewEnumerator1, Loop Duration Mode=::NewEnumerator1, UseLoopDelay=false)
   &SpawnBurst_Instantaneous(
    Spawn Count=&Multiply_Int(A=7, B=&MakeIntFromFloat(Float=$User.Variant))
    Use Spawn Probability=false
    UseLoopCountLimit=false
   )
  ]
  particleSpawnScript [
   &FN_SetSpawnGroup(Number of Spawn Groups=7)
   &InitializeParticle(
    Color=<rgba 0 1 0.1014 0.15>
    Sprite Size Mode=::NewEnumerator3
    Uniform Sprite Size=/Game/Effects/Niagara/DynamicInputs/FN_PickFloatFromSpawnGroupCutoff(Over Cutoff Float:NiagaraFloat=11, Under Cutoff Float:NiagaraFloat=14, SpawnGroup:NiagaraInt32=$Particles.SpawnGroup, Cutoff:NiagaraInt32=5)
    UsePositionOffset=false
    Write Material Random=true
   )
   &FN_Sprite_Attributes(
    Color=<rgba 0 1.5 0.1521 1>
    Create Color Attribute=true
    Create SpriteSize Attribute=true
    SpriteSize=&Multiply_Vector2D_ByFloat(Float=0.5, Vector2D=$Particles.SpriteSize)
    Write to Module Namespace=true
   )
  ]
  particleUpdateScript [
   &ParticleState(Kill Particles When Lifetime Has Elapsed=false, Let Infinitely Lived Particles Die When Emitter Deactivates=true)
   &SkeletalMeshLocation(
    Bone / Socket Index=$Particles.SpawnGroup
    MaskSpawnGroup=false
    Mesh Sampling Type=::NewEnumerator5
    Skeletal Mesh=$User.New Skeletal Mesh
    Socket Sampling Mode=::NewEnumerator5
   )
   &FN_Light_Attributes(
    Color=$Particles.Color
    Create Color Attribute=true
    Create Exponent Attribute=true
    Create Radius Attribute=true
    Exponent=5
    Radius=12.5
    Scale Color=6
   )
   &CameraOffset(Camera Offset Amount=1)
  ]
  rendererClasses [&NiagaraSpriteRendererProperties, &NiagaraLightRendererProperties]
  renderers [
   {rendererClass=&NiagaraSpriteRendererProperties}
   {rendererClass=&NiagaraLightRendererProperties, rendererIndex=1}
  ]
 }}
summary {
 systemName=NS_ThriveSpirit_Glows
 emitters [{emitterName=Necklace_Glow, rendererClasses [&NiagaraSpriteRendererProperties, &NiagaraLightRendererProperties]}]
 userVariables [
  User.New Skeletal Mesh {
   type=NiagaraDataInterfaceSkeletalMesh
   defaultValue {
    struct=NiagaraExt_VariableValue_DataInterface
    value {dataInterface=&self:NiagaraDataInterfaceSkeletalMesh_0}
   }
  }
  User.Variant {type=NiagaraFloat, defaultValue=1, description=""}
 ]
}
