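# Niagara system for a green glow on the head hair of the female Thrive Spirit character cosmetic, judging by its name and path.
# A single emitter bursts User.Variant particles (truncated to an integer, zero by default) at a fixed offset from the system location, persisting until deactivation.
# Two sprite renderers and a light renderer draw the camera-offset glow, with module-written attributes providing a brighter half-size sprite variant and the light's settings.
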NiagaraSystem /BRCosmetics/Characters/Player/Female/Medium/Bodies/F_MED_Thrive_Spirit/FX/NS_ThriveSpirit_HeadHair_Glows
dependencies {usedDataInterfaces [], usedDynamicInputs [], usedModules [&SystemState]}
emitters {Empty {
  emitterUpdateScript [
   &EmitterState(
    Life Cycle Mode=::NewEnumerator1
    Loop Behavior=::NewEnumerator1
    Loop Duration Mode=::NewEnumerator1
    Reset Age On Awaken=true
    Scalability Mode=::NewEnumerator1
    UseLoopDelay=false
   )
   &SpawnBurst_Instantaneous(
    Spawn Count=&Multiply_Int(A=1, B=&MakeIntFromFloat(Float=$User.Variant))
    Use Spawn Probability=false
    UseLoopCountLimit=false
   )
  ]
  particleSpawnScript [
   &InitializeParticle(
    Color=<rgba 0 1 0.1014 0.15>
    Lifetime=5
    Sprite Size Mode=::NewEnumerator3
    Uniform Sprite Size=14
    UsePositionOffset=false
    Write Material Random=true
   )
   &SystemLocation(MaskSpawnGroup=false, Offset=<11 9.5 0>)
   &FN_Sprite_Attributes(
    Color=<rgba 0 1.5 0.1521 1>
    Create Color Attribute=true
    Create SpriteSize Attribute=true
    SpriteSize=&Multiply_Vector2D_ByFloat(Float=0.5, Vector2D=$Particles.SpriteSize)
    Write to Module Namespace=true
   )
  ]
  particleUpdateScript [
   &ParticleState(Kill Particles When Lifetime Has Elapsed=false, Let Infinitely Lived Particles Die When Emitter Deactivates=true)
   &CameraOffset(Camera Offset Amount=1)
   &FN_Light_Attributes(
    Color=$Particles.Color
    Create Color Attribute=true
    Create Exponent Attribute=true
    Create Radius Attribute=true
    Exponent=5
    Radius=10
    Scale Color=6
   )
  ]
  rendererClasses [&NiagaraSpriteRendererProperties, &NiagaraLightRendererProperties]
  renderers [
   {rendererClass=&NiagaraSpriteRendererProperties}
   {rendererClass=&NiagaraSpriteRendererProperties, rendererIndex=1}
   {rendererClass=&NiagaraLightRendererProperties, rendererIndex=2}
  ]
 }}
summary {
 systemName=NS_ThriveSpirit_HeadHair_Glows
 emitters [{emitterName=Empty, rendererClasses [&NiagaraSpriteRendererProperties, &NiagaraLightRendererProperties]}]
 userVariables [User.Variant {type=NiagaraFloat, defaultValue=0, description=""}]
}
