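# Control Rig for the F_MED_Thrive_Spirit female player cosmetic body, adding procedural secondary motion to its arm bands and belts.
# Execution runs four Chain Harmonics units in sequence on the left and right arm band and two belt chains, combining reach, wave and pendulum, with debug draw on.
# A parent constraint blending dyn_necklace_05 between its neighbors sits outside the execution chain.
# Skirt and other dyn bones are not driven here.
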ControlRigBlueprint /BRCosmetics/Characters/Player/Female/Medium/Bodies/F_MED_Thrive_Spirit/Meshes/F_MED_Thrive_Spirit_CtrlRig
graphs [
 RigVMFunctionLibrary {links [], nodes []}
 RigVMModel {nodes [
   RigUnit_BeginExecution = &RigUnit_BeginExecution()
   /Script/ControlRig.RigUnit_ChainHarmonicsPerItem(
    ChainRoot=(Type=Bone,Name="dyn_arm_band_l_01")
    DrawWorldOffset=(Rotation=(X=0,Y=0,Z=0,W=1),Translation=(X=0,Y=0,Z=0),Scale3D=(X=1,Y=1,Z=1))
    Pendulum=(bEnabled=True,PendulumStiffness=2,PendulumGravity=(X=0,Y=0,Z=0),PendulumBlend=0.75,PendulumDrag=0.98,PendulumMaximum=0.1,UnwindAxis=(X=0,Y=1,Z=0),UnwindMinimum=0.2,UnwindMaximum=0.05)
    Reach=(bEnabled=True,ReachTarget=(X=0,Y=0,Z=0),ReachAxis=(X=1,Y=0,Z=0))
    Speed=(X=1,Y=1,Z=1)
    Wave=(bEnabled=True,WaveFrequency=(X=0,Y=10,Z=0),WaveAmplitude=(X=0,Y=0.8,Z=0),WaveOffset=(X=0,Y=5,Z=0),WaveNoise=(X=0,Y=0,Z=0),WaveMaximum=1.25)
    WaveCurve=(EditorCurveData=(Keys=((),(Time=1,Value=1)),DefaultValue=340282346638528859811704183484516925440,PreInfinityExtrap=RCCE_Constant,PostInfinityExtrap=RCCE_Constant))
    bDrawDebug=True
    ExecutePin=@RigUnit_BeginExecution.ExecutePin
   ) {is=struct RigVMUnitNode}
   > &RigUnit_ChainHarmonicsPerItem(
    ChainRoot=(Type=Bone,Name="dyn_arm_band_r_01")
    DrawWorldOffset=(Rotation=(X=0,Y=0,Z=0,W=1),Translation=(X=0,Y=0,Z=0),Scale3D=(X=1,Y=1,Z=1))
    Pendulum=(bEnabled=True,PendulumStiffness=2,PendulumGravity=(X=0,Y=0,Z=0),PendulumBlend=0.75,PendulumDrag=0.98,PendulumMaximum=0.1,UnwindAxis=(X=0,Y=1,Z=0),UnwindMinimum=0.2,UnwindMaximum=0.05)
    Reach=(bEnabled=True,ReachTarget=(X=0,Y=0,Z=0),ReachAxis=(X=1,Y=0,Z=0))
    Speed=(X=1,Y=1,Z=1)
    Wave=(bEnabled=True,WaveFrequency=(X=0,Y=10,Z=0),WaveAmplitude=(X=0,Y=0.8,Z=0),WaveOffset=(X=0,Y=5,Z=0),WaveNoise=(X=0,Y=0,Z=0),WaveMaximum=1.25)
    WaveCurve=(EditorCurveData=(Keys=((),(Time=1,Value=1)),DefaultValue=340282346638528859811704183484516925440,PreInfinityExtrap=RCCE_Constant,PostInfinityExtrap=RCCE_Constant))
    bDrawDebug=True
   )
   > &RigUnit_ChainHarmonicsPerItem(
    ChainRoot=(Type=Bone,Name="dyn_belt_a_02")
    DrawWorldOffset=(Rotation=(X=0,Y=0,Z=0,W=1),Translation=(X=0,Y=0,Z=0),Scale3D=(X=1,Y=1,Z=1))
    Pendulum=(bEnabled=True,PendulumStiffness=2,PendulumGravity=(X=0,Y=0,Z=0),PendulumBlend=0.75,PendulumDrag=0.98,PendulumMaximum=0.1,UnwindAxis=(X=0,Y=1,Z=0),UnwindMinimum=0.2,UnwindMaximum=0.05)
    Reach=(bEnabled=True,ReachTarget=(X=0,Y=0,Z=0),ReachAxis=(X=1,Y=0,Z=0))
    Speed=(X=1,Y=1,Z=1)
    Wave=(bEnabled=True,WaveFrequency=(X=0,Y=-2,Z=0),WaveAmplitude=(X=0,Y=-0.5,Z=0),WaveOffset=(X=0,Y=-2,Z=0),WaveNoise=(X=0,Y=0,Z=0),WaveMaximum=0.75)
    WaveCurve=(EditorCurveData=(Keys=((),(Time=1,Value=1)),DefaultValue=340282346638528859811704183484516925440,PreInfinityExtrap=RCCE_Constant,PostInfinityExtrap=RCCE_Constant))
    bDrawDebug=True
   )
   > &RigUnit_ChainHarmonicsPerItem(
    ChainRoot=(Type=Bone,Name="dyn_belt_b_01")
    DrawWorldOffset=(Rotation=(X=0,Y=0,Z=0,W=1),Translation=(X=0,Y=0,Z=0),Scale3D=(X=1,Y=1,Z=1))
    Pendulum=(bEnabled=True,PendulumStiffness=2,PendulumGravity=(X=0,Y=0,Z=0),PendulumBlend=0.75,PendulumDrag=0.98,PendulumMaximum=0.1,UnwindAxis=(X=0,Y=1,Z=0),UnwindMinimum=0.2,UnwindMaximum=0.05)
    Reach=(bEnabled=True,ReachTarget=(X=0,Y=0,Z=0),ReachAxis=(X=1,Y=0,Z=0))
    Speed=(X=1,Y=1,Z=1)
    Wave=(bEnabled=True,WaveFrequency=(X=0,Y=-3,Z=0),WaveAmplitude=(X=0,Y=-1,Z=0),WaveOffset=(X=0,Y=-2,Z=0),WaveNoise=(X=0,Y=0,Z=0),WaveMaximum=1.25)
    WaveCurve=(EditorCurveData=(Keys=((),(Time=1,Value=1)),DefaultValue=340282346638528859811704183484516925440,PreInfinityExtrap=RCCE_Constant,PostInfinityExtrap=RCCE_Constant))
    bDrawDebug=True
   )
   &RigUnit_ParentConstraint(
    AdvancedSettings=(RotationOrderForFilter=XZY)
    Child=(Type=Bone,Name="dyn_necklace_05")
    ExecutePin=()
    Parents=((Item=(Type=Bone,Name="dyn_necklace_04"),Weight=0.5),(Item=(Type=Bone,Name="dyn_necklace_06"),Weight=0.5))
    Weight=1
   )
  ]}
]
hierarchy [
 root {loc=<0 0 0>, rot=<0 0 0 1>}
 pelvis {parent=root, loc=<0 -0.5301 98.0571>, rot=<0.0127 0.707 -0.0127 -0.707>}
 spine_01 {parent=pelvis, loc=<6.1208 -0.9328 0>, rot=<0 0 0.034 -0.9994>}
 spine_02 {parent=spine_01, loc=<6.5996 -0.2095 0>, rot=<0 0 0.0238 -0.9997>}
 spine_03 {parent=spine_02, loc=<6.4914 -0.1237 0>, rot=<0 0 0.0149 -0.9999>}
 spine_04 {parent=spine_03, loc=<6.9598 0.0794 0>, rot=<0 0 -0.0132 -0.9999>}
 spine_05 {parent=spine_04, loc=<6.9987 0 0>, rot=<0 0 -0.0015 -1>}
 clavicle_l {parent=spine_05, loc=<5.4831 -0.1053 -3.2946>, rot=<0.0465 -0.7289 -0.009 -0.6829>}
 upperarm_l {parent=clavicle_l, loc=<9.8658 -1.2136 -0.4105>, rot=<0.0164 -0.3245 -0.0269 -0.9454>}
 lowerarm_l {parent=upperarm_l, loc=<26.465 -0.8345 0.259>, rot=<0.0036 0.0161 -0.2298 -0.9731>}
 hand_l {parent=lowerarm_l, loc=<22.7355 -0.3203 -0.6105>, rot=<-0.6784 -0.0223 0.0735 -0.7307>}
 index_metacarpal_l {parent=hand_l, loc=<1.6535 0.0183 -2.96>, rot=<0.0158 -0.0124 -0.0173 -0.9996>}
 index_01_l {parent=index_metacarpal_l, loc=<8.3008 0.0357 -0.0524>, rot=<-0.0044 0.0485 0.0924 -0.9945>}
 index_02_l {parent=index_01_l, loc=<3.3587 -0.3004 0.0452>, rot=<-0.002 0.0027 0.3225 -0.9466>}
 index_03_l {parent=index_02_l, loc=<2.5578 -0.123 -0.0252>, rot=<-0.002 0.0067 0.2558 -0.9667>}
 middle_metacarpal_l {parent=hand_l, loc=<2.1856 -0.1725 -0.5984>, rot=<0.006 0.0168 -0.0329 -0.9993>}
 middle_01_l {parent=middle_metacarpal_l, loc=<7.3275 -0.2447 0.0129>, rot=<-0.0208 0.0042 0.0858 -0.9961>}
 middle_02_l {parent=middle_01_l, loc=<4.0123 0.1401 0.1175>, rot=<-0.0218 0.0258 0.3512 -0.9357>}
 middle_03_l {parent=middle_02_l, loc=<2.6077 0.0534 0.1373>, rot=<0.0082 0.0188 0.1708 -0.9851>}
 pinky_metacarpal_l {parent=hand_l, loc=<1.7616 -0.9753 3.3814>, rot=<-0.0397 0.0814 -0.069 -0.9935>}
 pinky_01_l {parent=pinky_metacarpal_l, loc=<7.1709 -0.4831 -0.1094>, rot=<-0.0654 -0.0035 0.2242 -0.9723>}
 pinky_02_l {parent=pinky_01_l, loc=<2.8274 0.0893 -0.1043>, rot=<0.0037 0.0011 0.2854 -0.9584>}
 pinky_03_l {parent=pinky_02_l, loc=<2.2046 0.0307 -0.0078>, rot=<-0.0002 0.0098 0.1445 -0.9895>}
 ring_metacarpal_l {parent=hand_l, loc=<2.2074 -0.3625 1.3638>, rot=<-0.0284 0.0722 -0.0458 -0.9959>}
 ring_01_l {parent=ring_metacarpal_l, loc=<7.0295 -0.6197 -0.1446>, rot=<-0.0489 -0.0252 0.1892 -0.9804>}
 ring_02_l {parent=ring_01_l, loc=<3.6174 0.4647 -0.1027>, rot=<0 0 0.2806 -0.9598>}
 ring_03_l {parent=ring_02_l, loc=<2.4152 0.114 0.0199>, rot=<0 0 0.1189 -0.9929>}
 thumb_01_l {parent=hand_l, loc=<1.3118 -0.3191 -2.459>, rot=<0.6546 -0.4042 0.0664 -0.6354>}
 thumb_02_l {parent=thumb_01_l, loc=<5.9336 -0.0448 0.0401>, rot=<-0.0037 0.0213 0.298 -0.9543>}
 thumb_03_l {parent=thumb_02_l, loc=<2.2223 0.3682 -0.0907>, rot=<0.0118 -0.013 0.0066 -0.9998>}
 lowerarm_twist_01_l {parent=lowerarm_l, loc=<18.316 0 0>, rot=<0 0 0 1>}
 lowerarm_twist_02_l {parent=lowerarm_l, loc=<10.6707 0 0>, rot=<0 0 0 1>}
 upperarm_twist_01_l {parent=upperarm_l, loc=<7.2135 -0.1134 0>, rot=<0 0 0 1>}
 dyn_arm_band_l_MSTR {parent=upperarm_twist_01_l, loc=<0 0 0>, rot=<0 0 0 1>}
 dyn_arm_band_l_01 {parent=dyn_arm_band_l_MSTR, loc=<0.0209 5.1793 -4.2247>, rot=<0.5964 0.5736 0.3059 -0.4708>}
 dyn_arm_band_l_02 {parent=dyn_arm_band_l_01, loc=<7.7748 0 0>, rot=<-0.1091 -0.4919 0.0026 -0.8638>}
 dyn_arm_band_l_03 {parent=dyn_arm_band_l_02, loc=<7.7823 0 0>, rot=<-0.0449 -0.3505 -0.0449 -0.9344>}
 dyn_arm_band_l_04 {parent=dyn_arm_band_l_03, loc=<9.0365 0 0>, rot=<-0.0331 -0.6017 0.0022 -0.798>}
 dyn_arm_band_l_05 {parent=dyn_arm_band_l_04, loc=<9.6442 0 0>, rot=<0.0027 -0.4606 -0.261 -0.8484>}
 dyn_arm_band_l_06 {parent=dyn_arm_band_l_05, loc=<5.9234 0 0>, rot=<-0.489 -0.7849 0.3002 0.2341>}
 upperarm_twist_02_l {parent=upperarm_l, loc=<19.7998 -0.1134 0>, rot=<0 0 0 1>}
 dyn_l_cuff_MSTR {parent=upperarm_l, loc=<19.7554 2.6879 -0.3828>, rot=<-0.0065 0.3563 0.0039 -0.9343>}
 dyn_l_cuff_01 {parent=dyn_l_cuff_MSTR, loc=<0 0 0>, rot=<-0.0722 -0.3332 -0.2199 -0.914>}
 dyn_l_cuff_02_MSTR {parent=upperarm_l, loc=<19.7365 -4.6025 -0.4921>, rot=<-0.0065 0.3563 0.0039 -0.9343>}
 dyn_l_cuff_02 {parent=dyn_l_cuff_02_MSTR, loc=<0 0 0>, rot=<0.0165 -0.3405 0.0242 -0.9398>}
 dyn_necklace_MSTR_b {parent=clavicle_l, loc=<-2.0244 -7.3445 2.3156>, rot=<-0.5033 -0.495 0.4706 -0.5293>}
 dyn_necklace_09 {parent=dyn_necklace_MSTR_b, loc=<0 0 0>, rot=<-0.4983 -0.5303 0.5523 0.4069>}
 dyn_necklace_08 {parent=dyn_necklace_09, loc=<10.013 0 0>, rot=<0.0089 0.1764 -0.3748 -0.9101>}
 dyn_necklace_07 {parent=dyn_necklace_08, loc=<7.8925 0 0>, rot=<0 0 -0.7776 -0.6288>}
 dyn_necklace_06 {parent=dyn_necklace_07, loc=<11.7658 0 0>, rot=<0 0 0 1>}
 dyn_necklace_offset_06 {parent=dyn_necklace_06, loc=<0 0 0>, rot=<0.5056 0.0762 0.83 -0.2228>}
 dyn_necklace_offset_07 {parent=dyn_necklace_07, loc=<0 0 0>, rot=<0.5056 0.0762 0.83 -0.2228>}
 dyn_necklace_offset_08 {parent=dyn_necklace_08, loc=<0 0 0>, rot=<0.2586 0.4411 0.3487 -0.7855>}
 clavicle_r {parent=spine_05, loc=<5.4831 -0.1053 3.2946>, rot=<0.7289 0.0465 -0.6829 0.009>}
 upperarm_r {parent=clavicle_r, loc=<-9.8658 1.2136 0.4104>, rot=<0.0164 -0.3245 -0.0269 -0.9454>}
 lowerarm_r {parent=upperarm_r, loc=<-26.4651 0.8345 -0.259>, rot=<0.0036 0.0161 -0.2298 -0.9731>}
 hand_r {parent=lowerarm_r, loc=<-22.7356 0.3203 0.6106>, rot=<-0.6784 -0.0223 0.0735 -0.7307>}
 index_metacarpal_r {parent=hand_r, loc=<-1.6534 -0.0181 2.9601>, rot=<0.0158 -0.0124 -0.0173 -0.9996>}
 index_01_r {parent=index_metacarpal_r, loc=<-8.3008 -0.0357 0.0524>, rot=<-0.0044 0.0485 0.0924 -0.9945>}
 index_02_r {parent=index_01_r, loc=<-3.3589 0.3003 -0.0453>, rot=<-0.002 0.0027 0.3225 -0.9466>}
 index_03_r {parent=index_02_r, loc=<-2.5578 0.123 0.0252>, rot=<-0.002 0.0067 0.2558 -0.9667>}
 middle_metacarpal_r {parent=hand_r, loc=<-2.1856 0.1726 0.5984>, rot=<0.006 0.0168 -0.0329 -0.9993>}
 middle_01_r {parent=middle_metacarpal_r, loc=<-7.3273 0.2447 -0.0129>, rot=<-0.0208 0.0042 0.0858 -0.9961>}
 middle_02_r {parent=middle_01_r, loc=<-4.0124 -0.14 -0.1175>, rot=<-0.0218 0.0258 0.3512 -0.9357>}
 middle_03_r {parent=middle_02_r, loc=<-2.6077 -0.0533 -0.1373>, rot=<0.0082 0.0188 0.1708 -0.9851>}
 pinky_metacarpal_r {parent=hand_r, loc=<-1.7617 0.9753 -3.3814>, rot=<-0.0397 0.0814 -0.069 -0.9935>}
 pinky_01_r {parent=pinky_metacarpal_r, loc=<-7.1707 0.4832 0.1094>, rot=<-0.0654 -0.0035 0.2242 -0.9723>}
 pinky_02_r {parent=pinky_01_r, loc=<-2.8274 -0.0892 0.1043>, rot=<0.0037 0.0011 0.2854 -0.9584>}
 pinky_03_r {parent=pinky_02_r, loc=<-2.2045 -0.031 0.0078>, rot=<-0.0002 0.0098 0.1445 -0.9895>}
 ring_metacarpal_r {parent=hand_r, loc=<-2.2073 0.3627 -1.3638>, rot=<-0.0284 0.0722 -0.0458 -0.9959>}
 ring_01_r {parent=ring_metacarpal_r, loc=<-7.0297 0.6197 0.1446>, rot=<-0.0489 -0.0252 0.1892 -0.9804>}
 ring_02_r {parent=ring_01_r, loc=<-3.6171 -0.4648 0.1027>, rot=<0 0 0.2806 -0.9598>}
 ring_03_r {parent=ring_02_r, loc=<-2.4153 -0.1141 -0.0199>, rot=<0 0 0.1189 -0.9929>}
 thumb_01_r {parent=hand_r, loc=<-1.3119 0.3192 2.459>, rot=<0.6546 -0.4042 0.0664 -0.6354>}
 thumb_02_r {parent=thumb_01_r, loc=<-5.9335 0.0447 -0.0401>, rot=<-0.0037 0.0213 0.298 -0.9543>}
 thumb_03_r {parent=thumb_02_r, loc=<-2.2223 -0.3682 0.0907>, rot=<0.0118 -0.013 0.0066 -0.9998>}
 lowerarm_twist_01_r {parent=lowerarm_r, loc=<-18.3161 0 0>, rot=<0 0 0 1>}
 lowerarm_twist_02_r {parent=lowerarm_r, loc=<-10.6707 0 0>, rot=<0 0 0 1>}
 upperarm_twist_01_r {parent=upperarm_r, loc=<-7.2135 0.1134 0>, rot=<0 0 0 1>}
 dyn_arm_band_r_MSTR {parent=upperarm_twist_01_r, loc=<0 0 0>, rot=<0 0 0 -1>}
 dyn_arm_band_r_01 {parent=dyn_arm_band_r_MSTR, loc=<-0.0211 -5.1792 4.2247>, rot=<-0.3372 -0.4371 0.6216 -0.5558>}
 dyn_arm_band_r_02 {parent=dyn_arm_band_r_01, loc=<7.7746 0 0>, rot=<0.265 -0.4904 -0.0385 -0.8293>}
 dyn_arm_band_r_03 {parent=dyn_arm_band_r_02, loc=<7.7824 0 0>, rot=<-0.0027 -0.3523 -0.0277 -0.9355>}
 dyn_arm_band_r_04 {parent=dyn_arm_band_r_03, loc=<9.0363 0 0>, rot=<-0.0732 -0.6015 -0.0153 -0.7953>}
 dyn_arm_band_r_05 {parent=dyn_arm_band_r_04, loc=<9.6438 0 0>, rot=<-0.0025 -0.4288 0.3104 -0.8484>}
 dyn_arm_band_r_06 {parent=dyn_arm_band_r_05, loc=<5.9235 0 0>, rot=<0.8761 0 0 -0.4821>}
 upperarm_twist_02_r {parent=upperarm_r, loc=<-19.7998 0.1134 0>, rot=<0 0 0 1>}
 dyn_r_cuff_MSTR {parent=upperarm_r, loc=<-19.7554 -2.6879 0.3829>, rot=<0.9343 -0.0039 0.3563 -0.0065>}
 dyn_r_cuff_01 {parent=dyn_r_cuff_MSTR, loc=<0 0 0>, rot=<-0.3332 -0.0722 0.914 0.2199>}
 dyn_r_cuff_02_MSTR {parent=upperarm_r, loc=<-19.7362 4.7243 0.494>, rot=<0.9343 -0.0039 0.3563 -0.0065>}
 dyn_r_cuff_02 {parent=dyn_r_cuff_02_MSTR, loc=<0 0 0>, rot=<-0.3405 0.0165 0.9398 -0.0242>}
 dyn_necklace_MSTR {parent=clavicle_r, loc=<2.065 7.3466 -2.3128>, rot=<-0.4706 -0.5293 -0.5033 0.495>}
 dyn_necklace_01 {parent=dyn_necklace_MSTR, loc=<0 0 0>, rot=<0.649 -0.6947 -0.2793 0.1347>}
 dyn_necklace_02 {parent=dyn_necklace_01, loc=<10.3214 0 0>, rot=<0.5578 0.2776 0.3574 -0.6958>}
 dyn_necklace_03 {parent=dyn_necklace_02, loc=<7.7814 0 0>, rot=<-0.0317 -0.296 0.6874 -0.6625>}
 dyn_necklace_04 {parent=dyn_necklace_03, loc=<9.9501 0 0>, rot=<0.0093 -0.0564 -0.0216 -0.9981>}
 dyn_necklace_offset_04 {parent=dyn_necklace_04, loc=<0 0 0>, rot=<0.1226 0.8217 0.2185 -0.5119>}
 dyn_necklace_offset_03 {parent=dyn_necklace_03, loc=<0 0 0>, rot=<0.1217 0.796 0.1924 -0.5609>}
 dyn_necklace_offset_02 {parent=dyn_necklace_02, loc=<0 0 0>, rot=<0.6669 0.2716 0.5022 -0.4788>}
 neck_01 {parent=spine_05, loc=<11.3138 0.1529 0>, rot=<0 0 -0.2003 -0.9797>}
 neck_02 {parent=neck_01, loc=<4.5627 0 0>, rot=<0 0 0.0504 -0.9987>}
 head {parent=neck_02, loc=<5.3914 0 0>, rot=<0 0 0.115 -0.9934>}
 dyn_necklace_05 {parent=spine_05, loc=<4.2639 8.408 -0.3834>, rot=<0.1763 -0.4786 0.1645 -0.8443>}
 dyn_necklace_offset_05 {parent=dyn_necklace_05, loc=<0 0 0>, rot=<-0.157 0.4853 -0.1982 -0.837>}
 dyn_necklace_helper {parent=spine_05, loc=<9.1595 -6.4513 0>, rot=<0 0 0 1>}
 dyn_belt_B_MSTR {parent=spine_02, loc=<2.2482 -1.3345 -9.6874>, rot=<-0.1069 -0.8521 0.4293 -0.2798>}
 dyn_belt_b_01 {parent=dyn_belt_B_MSTR, loc=<3.3569 0 0>, rot=<-0.2387 0.0997 0.0526 -0.9645>}
 dyn_belt_b_02 {parent=dyn_belt_b_01, loc=<7.5594 0 0>, rot=<-0.3851 0.1739 0.0203 -0.9061>}
 dyn_belt_b_03 {parent=dyn_belt_b_02, loc=<9.2115 0 0>, rot=<-0.073 0.032 0.2136 -0.9737>}
 dyn_belt_b_04 {parent=dyn_belt_b_03, loc=<8.121 0 0>, rot=<0 0 0.1956 -0.9807>}
 dyn_belt_a_MSTR {parent=spine_02, loc=<2.275 -2.2254 -9.0318>, rot=<-0.0162 -0.1157 0.852 -0.5104>}
 dyn_belt_a_02 {parent=dyn_belt_a_MSTR, loc=<4.7099 0 0>, rot=<0.1893 -0.1125 0.0984 -0.9705>}
 dyn_belt_a_03 {parent=dyn_belt_a_02, loc=<6.5122 0 0>, rot=<0.3525 -0.1599 0.0943 -0.9172>}
 dyn_belt_a_04 {parent=dyn_belt_a_03, loc=<8.0551 0 0>, rot=<0.2174 -0.065 0.1447 -0.9631>}
 dyn_belt_a_05 {parent=dyn_belt_a_04, loc=<10.7977 0 0>, rot=<0 0 0.1365 -0.9906>}
 thigh_l {parent=pelvis, loc=<-1.0855 -0.2109 -8.7053>, rot=<0.0009 0.0355 0.0155 -0.9993>}
 calf_l {parent=thigh_l, loc=<-42.5284 -0.0019 -0.7297>, rot=<0.0006 -0.0161 -0.0391 -0.9991>}
 calf_twist_01_l {parent=calf_l, loc=<-35.8483 -1.9197 -0.5794>, rot=<0.0237 0.0095 -0.0035 -0.9997>}
 calf_twist_02_l {parent=calf_l, loc=<-19.1546 -1.7943 -0.2638>, rot=<0.0237 0.0095 -0.0035 -0.9997>}
 foot_l {parent=calf_l, loc=<-46.7402 -2.5014 -0.592>, rot=<0.0427 -0.0238 0.0353 -0.9982>}
 ball_l {parent=foot_l, loc=<-3.5658 10.8066 -0.1544>, rot=<0.0001 -0.0001 -0.7063 -0.7079>}
 thigh_twist_01_l {parent=thigh_l, loc=<-27.952 0 0>, rot=<0.0013 0 0 -1>}
 dyn_skirtFront_mstr_l {parent=thigh_l, loc=<2.9384 2.617 -4.0464>, rot=<-0.0201 -0.9976 -0.0512 -0.0424>}
 dyn_skirtFront_01_l {parent=dyn_skirtFront_mstr_l, loc=<0 0 0>, rot=<0 0 0 1>, scale=<0.7498 1 1>}
 dyn_skirtFront_02_l {parent=dyn_skirtFront_01_l, loc=<17 0 0>, rot=<0.0172 -0.0001 0.0118 -0.9998>, scale=<0.9998 1.0004 1>}
 dyn_skirtFront_03_l {parent=dyn_skirtFront_02_l, loc=<17 0 0>, rot=<-0.0062 0.0007 0.0014 -1>}
 dyn_skirtFront_04_l {parent=dyn_skirtFront_03_l, loc=<17 0 0>, rot=<0.0093 -0.0023 -0.0025 -1>}
 dyn_skirtFront_05_l {parent=dyn_skirtFront_04_l, loc=<17 0 0>, rot=<0 0 0 1>, scale=<1.3336 1 1>}
 thigh_r {parent=pelvis, loc=<-1.0853 -0.2109 8.7053>, rot=<0.0355 -0.0009 0.9993 0.0155>}
 calf_r {parent=thigh_r, loc=<42.5286 0.0019 0.7297>, rot=<0.0006 -0.0161 -0.0391 -0.9991>}
 calf_twist_01_r {parent=calf_r, loc=<35.8485 1.9197 0.5794>, rot=<0.0237 0.0095 -0.0035 -0.9997>}
 calf_twist_02_r {parent=calf_r, loc=<19.1546 1.7943 0.2638>, rot=<0.0237 0.0095 -0.0035 -0.9997>}
 foot_r {parent=calf_r, loc=<46.7403 2.5014 0.592>, rot=<0.0427 -0.0238 0.0353 -0.9982>}
 ball_r {parent=foot_r, loc=<3.5658 -10.8066 0.1544>, rot=<0.0001 -0.0001 -0.7063 -0.7079>}
 thigh_twist_01_r {parent=thigh_r, loc=<27.952 0 0>, rot=<0.0013 0 0 -1>}
 dyn_skirtFront_mstr {parent=thigh_r, loc=<-3.4241 -9.4414 -7.5565>, rot=<-0.7113 0.0018 -0.0079 -0.7028>}
 dyn_skirtFront_01 {parent=dyn_skirtFront_mstr, loc=<0 0 0>, rot=<0 0 0 1>, scale=<0.7816 1 1>}
 dyn_skirtFront_02 {parent=dyn_skirtFront_01, loc=<17 0 0>, rot=<0.0544 0.0016 -0.0097 -0.9985>, scale=<0.9999 1.0002 1>}
 dyn_skirtFront_03 {parent=dyn_skirtFront_02, loc=<17 0 0>, rot=<0.0063 0.0009 -0.0013 -1>}
 dyn_skirtFront_04 {parent=dyn_skirtFront_03, loc=<17 0 0>, rot=<-0.0093 -0.0028 0.0021 -1>}
 dyn_skirtFront_05 {parent=dyn_skirtFront_04, loc=<17 0 0>, rot=<0 0 0 1>, scale=<1.2793 1 1>}
 dyn_skirtSide_mstr_l {parent=pelvis, loc=<2.0767 -7.2298 -12.8191>, rot=<-0.0026 -0.997 -0.0464 -0.0616>}
 dyn_skirtSide_01_l {parent=dyn_skirtSide_mstr_l, loc=<0 0 0>, rot=<0 0 0 1>, scale=<0.7518 1 1>}
 dyn_skirtSide_02_l {parent=dyn_skirtSide_01_l, loc=<17 0 0>, rot=<0.0172 -0.0001 0.0118 -0.9998>, scale=<0.9998 1.0004 1>}
 dyn_skirtSide_03_l {parent=dyn_skirtSide_02_l, loc=<17 0 0>, rot=<-0.0062 0.0007 0.0014 -1>}
 dyn_skirtSide_04_l {parent=dyn_skirtSide_03_l, loc=<17 0 0>, rot=<0.0093 -0.0023 -0.0025 -1>}
 dyn_skirtSide_05_l {parent=dyn_skirtSide_04_l, loc=<17 0 0>, rot=<0 0 0 1>, scale=<1.3301 1 1>}
 dyn_skirtBack_mstr_l {parent=pelvis, loc=<1.9672 -11.4919 -5.3852>, rot=<-0.0155 -0.7986 0.5993 -0.0539>}
 dyn_skirtBack_01_l {parent=dyn_skirtBack_mstr_l, loc=<0 0 0>, rot=<0 0 0.0219 -0.9998>, scale=<0.7785 1 1>}
 dyn_skirtBack_02_l {parent=dyn_skirtBack_01_l, loc=<17 0 0>, rot=<0.0271 0.0044 0.0001 -0.9996>}
 dyn_skirtBack_03_l {parent=dyn_skirtBack_02_l, loc=<17 0 0>, rot=<-0.0184 0.0075 -0.0001 -0.9998>, scale=<0.9999 1 1.0001>}
 dyn_skirtBack_04_l {parent=dyn_skirtBack_03_l, loc=<17 0 0>, rot=<-0.0016 -0.0002 0 -1>}
 dyn_skirtBack_05_l {parent=dyn_skirtBack_04_l, loc=<17 0 0>, rot=<0 0 0 1>, scale=<1.2845 1 1>}
 dyn_skirtBack_mstr_r {parent=pelvis, loc=<1.973 -11.4834 5.4628>, rot=<0.0408 -0.6022 0.7973 -0.0018>}
 dyn_skirtBack_01_r {parent=dyn_skirtBack_mstr_r, loc=<0 0 0>, rot=<0 0 -0.022 -0.9998>, scale=<0.7785 1 1>}
 dyn_skirtBack_02_r {parent=dyn_skirtBack_01_r, loc=<17 0 0>, rot=<-0.0271 0.0044 -0.0001 -0.9996>}
 dyn_skirtBack_03_r {parent=dyn_skirtBack_02_r, loc=<17 0 0>, rot=<-0.001 -0.0004 0 -1>}
 dyn_skirtBack_04_r {parent=dyn_skirtBack_03_r, loc=<17 0 0>, rot=<0.0016 -0.0001 0 -1>}
 dyn_skirtBack_05_r {parent=dyn_skirtBack_04_r, loc=<17 0 0>, rot=<0 0 0 1>, scale=<1.2845 1 1>}
 dyn_skirtSide_mstr_r {parent=pelvis, loc=<2.318 -4.2938 14.3557>, rot=<0.0609 -0.2002 0.9777 0.0162>}
 dyn_skirtSide_01_r {parent=dyn_skirtSide_mstr_r, loc=<0 0 0>, rot=<0 0 0 1>}
 dyn_skirtSide_02_r {parent=dyn_skirtSide_01_r, loc=<11.6729 0 0>, rot=<0.0544 0.002 -0.0124 -0.9984>}
 dyn_skirtSide_03_r {parent=dyn_skirtSide_02_r, loc=<11.6729 0 0>, rot=<0.0063 0.0012 -0.0016 -1>}
 dyn_skirtSide_04_r {parent=dyn_skirtSide_03_r, loc=<11.6729 0 0>, rot=<-0.0093 -0.0036 0.0027 -0.9999>}
 dyn_skirtSide_05_r {parent=dyn_skirtSide_04_r, loc=<11.6729 0 0>, rot=<0 0 0 1>}
 dyn_skirtFront_mstr_r {parent=pelvis, loc=<2.3502 5.7864 11.7956>, rot=<0.0785 0.3379 0.9379 0.0086>}
 dyn_skirtFront_01_r {parent=dyn_skirtFront_mstr_r, loc=<0 0 0>, rot=<0 0 0 1>}
 dyn_skirtFront_02_r {parent=dyn_skirtFront_01_r, loc=<11.754 0 0>, rot=<0.0544 0.002 -0.0124 -0.9984>}
 dyn_skirtFront_03_r {parent=dyn_skirtFront_02_r, loc=<11.754 0 0>, rot=<0.0063 0.0012 -0.0016 -1>}
 dyn_skirtFront_04_r {parent=dyn_skirtFront_03_r, loc=<11.754 0 0>, rot=<-0.0093 -0.0036 0.0027 -0.9999>}
 dyn_skirtFront_05_r {parent=dyn_skirtFront_04_r, loc=<11.754 0 0>, rot=<0 0 0 1>}
 dyn_glute_l {parent=pelvis, loc=<-8.179 -2.5701 -7.094>, rot=<-0.0127 -0.707 0.0127 -0.707>}
 dyn_glute_r {parent=pelvis, loc=<-8.179 -2.5701 7.094>, rot=<-0.0127 -0.707 0.0127 -0.707>}
 dyn_shirt_fixer_MSTR {parent=pelvis, loc=<7.5355 8.9234 2.3878>, rot=<0.0127 0.707 0.707 0.0127>}
 dyn_shirt_fixer {parent=dyn_shirt_fixer_MSTR, loc=<0 0 0>, rot=<0 0 0 1>}
 dyn_shirt_fixer_02_MSTR {parent=pelvis, loc=<3.4052 6.2992 -10.5018>, rot=<-0.0166 -0.9237 -0.3826 -0.0069>}
 dyn_shirt_fixer_02 {parent=dyn_shirt_fixer_02_MSTR, loc=<0 0 0>, rot=<0 0 0 1>}
 dyn_root_simspace {parent=root, loc=<0 0 0>, rot=<0 0 0 1>}
 dyn_floor_COL {parent=root, loc=<0 0 0>, rot=<0 0 0 1>}
]
