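# Physics asset for the female Thrive Spirit body's rope emote variant, covering dangling skirt and belt bone chains.
# Capsules on pelvis, spine, arms and legs form the body volume.
# Seven skirt chains hang from pelvis and two belt chains from spine_02, each using a tiny free-rotating master body and four angle-limited segments with tight twist.
# Two free-rotation cross-links tie neighbouring skirt chains at the third segment: front-to-front-right and front-left-to-side-left.
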PhysicsAsset /BRCosmetics/Characters/Player/Female/Medium/Bodies/F_MED_Thrive_Spirit/Meshes/F_MED_Thrive_Spirit_Rope_Emote_Physics
bodies {
 calf_l [{shapeType=Capsule, length=46.6211, radius=7.3375, center=<-28.7011 -0.7036 0.0466>, rotation=<rot 0 89.533 -88.5703>}]
 calf_r [{shapeType=Capsule, length=46.6211, radius=7.3375, center=<28.7258 0.7047 1.4779>, rotation=<rot -0.3385 89.533 -88.4634>}]
 dyn_belt_B_MSTR [{shapeType=Capsule, length=0.505, radius=0.505, shapeName=dyn_belt_B_MSTR_capsule, rotation=<rot 0 -90 0>}]
 dyn_belt_a_02 [{shapeType=Capsule, length=2, radius=2.5, shapeName=dyn_belt_a_02_capsule, center=<3 0 0>, rotation=<rot 90 0 0>}]
 dyn_belt_a_03 [{shapeType=Capsule, length=3.125, radius=2.875, shapeName=dyn_belt_a_03_capsule, center=<4.5 0 0>, rotation=<rot 90 0 0>}]
 dyn_belt_a_04 [{shapeType=Capsule, length=5.625, radius=2.875, shapeName=dyn_belt_a_04_capsule, center=<5.5 0 0>, rotation=<rot 90 0 0>}]
 dyn_belt_a_05 [{shapeType=Capsule, length=6.375, radius=3.125, shapeName=dyn_belt_a_05_capsule, center=<6 0 0>, rotation=<rot 90 0 0>}]
 dyn_belt_a_MSTR [{shapeType=Capsule, length=0.505, radius=0.505, shapeName=dyn_belt_a_MSTR_capsule, rotation=<rot 0 -90 0>}]
 dyn_belt_b_01 [{shapeType=Capsule, length=3.5, radius=2.5, shapeName=dyn_belt_b_01_capsule, center=<4 0 0>, rotation=<rot 90 0 0>}]
 dyn_belt_b_02 [{shapeType=Capsule, length=3.5, radius=3, shapeName=dyn_belt_b_02_capsule, center=<4.5 0 0>, rotation=<rot 90 0 0>}]
 dyn_belt_b_03 [{shapeType=Capsule, length=3, radius=3.5, shapeName=dyn_belt_b_03_capsule, center=<4 0 0>, rotation=<rot 90 0 0>}]
 dyn_belt_b_04 [{shapeType=Capsule, length=3, radius=3.5, shapeName=dyn_belt_b_04_capsule, center=<4.5 0 0>, rotation=<rot 90 0 0>}]
 dyn_skirtBack_01_l [{shapeType=Capsule, length=8, radius=6, shapeName=dyn_skirtBack_01_l_capsule, center=<5 0 -7>, rotation=<rot 90 0 0>}]
 dyn_skirtBack_01_r [{shapeType=Capsule, length=8, radius=6, shapeName=dyn_skirtBack_01_r_capsule, center=<5 0 -7>, rotation=<rot 90 0 0>}]
 dyn_skirtBack_02_l [{shapeType=Capsule, length=8, radius=6, shapeName=dyn_skirtBack_02_l_capsule, center=<5 0 -7>, rotation=<rot 90 0 0>}]
 dyn_skirtBack_02_r [{shapeType=Capsule, length=8, radius=6, shapeName=dyn_skirtBack_02_r_capsule, center=<5 0 -7>, rotation=<rot 90 0 0>}]
 dyn_skirtBack_03_l [{shapeType=Capsule, length=8, radius=6, shapeName=dyn_skirtBack_03_l_capsule, center=<5 0 -7>, rotation=<rot 90 0 0>}]
 dyn_skirtBack_03_r [{shapeType=Capsule, length=8, radius=6, shapeName=dyn_skirtBack_03_r_capsule, center=<5 0 -7>, rotation=<rot 90 0 0>}]
 dyn_skirtBack_04_l [{shapeType=Capsule, length=8, radius=6, shapeName=dyn_skirtBack_04_l_capsule, center=<5 0 -7>, rotation=<rot 90 0 0>}]
 dyn_skirtBack_04_r [{shapeType=Capsule, length=8, radius=6, shapeName=dyn_skirtBack_04_r_capsule, center=<5 0 -7>, rotation=<rot 90 0 0>}]
 dyn_skirtBack_mstr_l [{shapeType=Capsule, length=0.505, radius=0.505, shapeName=dyn_skirtBack_mstr_l_capsule, rotation=<rot 0 -90 0>}]
 dyn_skirtBack_mstr_r [{shapeType=Capsule, length=0.505, radius=0.505, shapeName=dyn_skirtBack_mstr_r_capsule, rotation=<rot 0 -90 0>}]
 dyn_skirtFront_01 [{shapeType=Capsule, length=8, radius=6, shapeName=dyn_skirtFront_01_capsule, center=<6 3 -7>, rotation=<rot 90 0 0>}]
 dyn_skirtFront_01_l [{shapeType=Capsule, length=8, radius=6, shapeName=dyn_skirtFront_01_l_capsule, center=<5 3 -7>, rotation=<rot 90 0 0>}]
 dyn_skirtFront_01_r [{shapeType=Capsule, length=8, radius=6, shapeName=dyn_skirtFront_01_r_capsule, center=<5 0 -7>, rotation=<rot 90 0 0>}]
 dyn_skirtFront_02 [{shapeType=Capsule, length=8, radius=6, shapeName=dyn_skirtFront_02_capsule, center=<6 3 -7>, rotation=<rot 90 0 0>}]
 dyn_skirtFront_02_l [{shapeType=Capsule, length=8, radius=6, shapeName=dyn_skirtFront_02_l_capsule, center=<5 3 -7>, rotation=<rot 90 0 0>}]
 dyn_skirtFront_02_r [{shapeType=Capsule, length=8, radius=6, shapeName=dyn_skirtFront_02_r_capsule, center=<5 0 -7>, rotation=<rot 90 0 0>}]
 dyn_skirtFront_03 [{shapeType=Capsule, length=8, radius=6, shapeName=dyn_skirtFront_03_capsule, center=<6 3 -7>, rotation=<rot 90 0 0>}]
 dyn_skirtFront_03_l [{shapeType=Capsule, length=8, radius=6, shapeName=dyn_skirtFront_03_l_capsule, center=<5 3 -7>, rotation=<rot 90 0 0>}]
 dyn_skirtFront_03_r [{shapeType=Capsule, length=8, radius=6, shapeName=dyn_skirtFront_03_r_capsule, center=<5 0 -7>, rotation=<rot 90 0 0>}]
 dyn_skirtFront_04 [{shapeType=Capsule, length=8, radius=6, shapeName=dyn_skirtFront_04_capsule, center=<6 3 -7>, rotation=<rot 90 0 0>}]
 dyn_skirtFront_04_l [{shapeType=Capsule, length=8, radius=6, shapeName=dyn_skirtFront_04_l_capsule, center=<5 3 -7>, rotation=<rot 90 0 0>}]
 dyn_skirtFront_04_r [{shapeType=Capsule, length=8, radius=6, shapeName=dyn_skirtFront_04_r_capsule, center=<5 0 -7>, rotation=<rot 90 0 0>}]
 dyn_skirtFront_mstr [{shapeType=Capsule, length=0.505, radius=0.505, shapeName=dyn_skirtFront_mstr_capsule}]
 dyn_skirtFront_mstr_l [{shapeType=Capsule, length=0.505, radius=0.505, shapeName=dyn_skirtFront_mstr_l_capsule, rotation=<rot 0 -90 0>}]
 dyn_skirtFront_mstr_r [{shapeType=Capsule, length=0.505, radius=0.505, shapeName=dyn_skirtFront_mstr_r_capsule, rotation=<rot 0 -90 0>}]
 dyn_skirtSide_01_l [{shapeType=Capsule, length=8, radius=6, shapeName=dyn_skirtSide_01_l_capsule, center=<5 3 -5>, rotation=<rot 90 0 0>}]
 dyn_skirtSide_01_r [{shapeType=Capsule, length=8, radius=6, shapeName=dyn_skirtSide_01_r_capsule, center=<5 3 -7>, rotation=<rot 90 0 0>}]
 dyn_skirtSide_02_l [{shapeType=Capsule, length=8, radius=6, shapeName=dyn_skirtSide_02_l_capsule, center=<5 3 -5>, rotation=<rot 90 0 0>}]
 dyn_skirtSide_02_r [{shapeType=Capsule, length=8, radius=6, shapeName=dyn_skirtSide_02_r_capsule, center=<5 3 -7>, rotation=<rot 90 0 0>}]
 dyn_skirtSide_03_l [{shapeType=Capsule, length=8, radius=6, shapeName=dyn_skirtSide_03_l_capsule, center=<5 3 -5>, rotation=<rot 90 0 0>}]
 dyn_skirtSide_03_r [{shapeType=Capsule, length=8, radius=6, shapeName=dyn_skirtSide_03_r_capsule, center=<5 3 -7>, rotation=<rot 90 0 0>}]
 dyn_skirtSide_04_l [{shapeType=Capsule, length=8, radius=6, shapeName=dyn_skirtSide_04_l_capsule, center=<5 3 -5>, rotation=<rot 90 0 0>}]
 dyn_skirtSide_04_r [{shapeType=Capsule, length=8, radius=6, shapeName=dyn_skirtSide_04_r_capsule, center=<5 3 -7>, rotation=<rot 90 0 0>}]
 dyn_skirtSide_mstr_l [{shapeType=Capsule, length=0.505, radius=0.505, shapeName=dyn_skirtSide_mstr_l_capsule, rotation=<rot 0 -90 0>}]
 dyn_skirtSide_mstr_r [{shapeType=Capsule, length=0.505, radius=0.505, shapeName=dyn_skirtSide_mstr_r_capsule, rotation=<rot 0 -90 0>}]
 lowerarm_l [{shapeType=Capsule, length=18.625, radius=5.13, shapeName=lowerarm_l_capsule, center=<12.7781 -0.0002 0>, rotation=<rot 90 0 0>}]
 lowerarm_r [{shapeType=Capsule, length=18.88, radius=5.005, shapeName=lowerarm_r_capsule, center=<-11.4315 0.0001 0>, rotation=<rot 90 0 0>}]
 pelvis [
  {shapeType=Capsule, length=8.5875, radius=9.6801, center=<-0.9227 -1.2812 0>}
  {shapeType=Capsule, length=13.415, radius=7.4702, shapeName=pelvis_capsule, center=<-3.9978 -3.7131 0>}
 ]
 spine_01 [{shapeType=Capsule, length=7.5011, radius=8.4272, shapeName=spine_01_capsule, center=<-0.8005 0.1286 -0.0001>}]
 spine_02 [{shapeType=Capsule, length=5.4764, radius=7.5222, shapeName=spine_02_capsule, center=<-2.0322 1.7597 0>}]
 thigh_l [{shapeType=Capsule, length=28.3361, radius=7.6472, center=<-21.4716 0.4992 -0.2548>, rotation=<rot 90 -7.5103 -7.3149>}]
 thigh_r [{shapeType=Capsule, length=27.08, radius=8.9033, center=<21.7868 -0.0552 -1.1637>, rotation=<rot 88.343 -180 -180>}]
 upperarm_l [{shapeType=Capsule, length=22.1346, radius=5.6454, shapeName=upperarm_l_capsule, center=<11.6225 -0.0001 0>, rotation=<rot 90 0 0>}]
 upperarm_r [{shapeType=Capsule, length=14.8754, radius=6.3899, shapeName=upperarm_r_capsule, center=<-11.6632 0.8314 0.6926>, rotation=<rot 90 0 0>}]
}
constraints [
 {bone1Name=thigh_l, bone2Name=pelvis, swing1LimitDegrees=45, swing2LimitDegrees=45, twistLimitDegrees=45}
 {bone1Name=calf_l, bone2Name=thigh_l, swing1LimitDegrees=45, swing2LimitDegrees=45, twistLimitDegrees=45}
 {bone1Name=thigh_r, bone2Name=pelvis, swing1LimitDegrees=45, swing2LimitDegrees=45, twistLimitDegrees=45}
 {bone1Name=calf_r, bone2Name=thigh_r, swing1LimitDegrees=45, swing2LimitDegrees=45, twistLimitDegrees=45}
 {
  bone1Name=dyn_skirtFront_mstr
  bone2Name=pelvis
  swing1LimitDegrees=45
  swing1Motion=Free
  swing2LimitDegrees=45
  swing2Motion=Free
  twistLimitDegrees=45
  twistMotion=Free
 }
 {bone1Name=dyn_skirtFront_01, bone2Name=dyn_skirtFront_mstr, swing1LimitDegrees=45, swing2LimitDegrees=45, twistLimitDegrees=15}
 {bone1Name=dyn_skirtFront_02, bone2Name=dyn_skirtFront_01, swing1LimitDegrees=45, swing2LimitDegrees=45, twistLimitDegrees=15}
 {bone1Name=dyn_skirtFront_03, bone2Name=dyn_skirtFront_02, swing1LimitDegrees=45, swing2LimitDegrees=45, twistLimitDegrees=15}
 {bone1Name=dyn_skirtFront_04, bone2Name=dyn_skirtFront_03, swing1LimitDegrees=45, swing2LimitDegrees=45, twistLimitDegrees=15}
 {bone1Name=dyn_skirtFront_04_r, bone2Name=dyn_skirtFront_03_r, swing1LimitDegrees=45, swing2LimitDegrees=45, twistLimitDegrees=15}
 {bone1Name=dyn_skirtFront_01_r, bone2Name=dyn_skirtFront_mstr_r, swing1LimitDegrees=45, swing2LimitDegrees=45, twistLimitDegrees=15}
 {bone1Name=dyn_skirtFront_02_r, bone2Name=dyn_skirtFront_01_r, swing1LimitDegrees=45, swing2LimitDegrees=45, twistLimitDegrees=15}
 {bone1Name=dyn_skirtFront_03_r, bone2Name=dyn_skirtFront_02_r, swing1LimitDegrees=45, swing2LimitDegrees=45, twistLimitDegrees=15}
 {
  bone1Name=dyn_skirtFront_mstr_r
  bone2Name=pelvis
  swing1LimitDegrees=45
  swing1Motion=Free
  swing2LimitDegrees=45
  swing2Motion=Free
  twistLimitDegrees=45
  twistMotion=Free
 }
 {bone1Name=dyn_skirtFront_04_l, bone2Name=dyn_skirtFront_03_l, swing1LimitDegrees=45, swing2LimitDegrees=45, twistLimitDegrees=15}
 {bone1Name=dyn_skirtFront_01_l, bone2Name=dyn_skirtFront_mstr_l, swing1LimitDegrees=45, swing2LimitDegrees=45, twistLimitDegrees=15}
 {bone1Name=dyn_skirtFront_02_l, bone2Name=dyn_skirtFront_01_l, swing1LimitDegrees=45, swing2LimitDegrees=45, twistLimitDegrees=15}
 {bone1Name=dyn_skirtFront_03_l, bone2Name=dyn_skirtFront_02_l, swing1LimitDegrees=45, swing2LimitDegrees=45, twistLimitDegrees=15}
 {
  bone1Name=dyn_skirtFront_mstr_l
  bone2Name=pelvis
  swing1LimitDegrees=45
  swing1Motion=Free
  swing2LimitDegrees=45
  swing2Motion=Free
  twistLimitDegrees=45
  twistMotion=Free
 }
 {bone1Name=dyn_skirtSide_04_r, bone2Name=dyn_skirtSide_03_r, swing1LimitDegrees=45, swing2LimitDegrees=45, twistLimitDegrees=15}
 {
  bone1Name=dyn_skirtSide_mstr_r
  bone2Name=pelvis
  swing1LimitDegrees=45
  swing1Motion=Free
  swing2LimitDegrees=45
  swing2Motion=Free
  twistLimitDegrees=45
  twistMotion=Free
 }
 {bone1Name=dyn_skirtSide_01_r, bone2Name=dyn_skirtSide_mstr_r, swing1LimitDegrees=45, swing2LimitDegrees=45, twistLimitDegrees=15}
 {bone1Name=dyn_skirtSide_02_r, bone2Name=dyn_skirtSide_01_r, swing1LimitDegrees=45, swing2LimitDegrees=45, twistLimitDegrees=15}
 {bone1Name=dyn_skirtSide_03_r, bone2Name=dyn_skirtSide_02_r, swing1LimitDegrees=45, swing2LimitDegrees=45, twistLimitDegrees=15}
 {bone1Name=dyn_skirtSide_04_l, bone2Name=dyn_skirtSide_03_l, swing1LimitDegrees=45, swing2LimitDegrees=45, twistLimitDegrees=15}
 {
  bone1Name=dyn_skirtSide_mstr_l
  bone2Name=pelvis
  swing1LimitDegrees=45
  swing1Motion=Free
  swing2LimitDegrees=45
  swing2Motion=Free
  twistLimitDegrees=45
  twistMotion=Free
 }
 {bone1Name=dyn_skirtSide_01_l, bone2Name=dyn_skirtSide_mstr_l, swing1LimitDegrees=45, swing2LimitDegrees=45, twistLimitDegrees=15}
 {bone1Name=dyn_skirtSide_02_l, bone2Name=dyn_skirtSide_01_l, swing1LimitDegrees=45, swing2LimitDegrees=45, twistLimitDegrees=15}
 {bone1Name=dyn_skirtSide_03_l, bone2Name=dyn_skirtSide_02_l, swing1LimitDegrees=45, swing2LimitDegrees=45, twistLimitDegrees=15}
 {bone1Name=dyn_skirtBack_04_r, bone2Name=dyn_skirtBack_03_r, swing1LimitDegrees=45, swing2LimitDegrees=45, twistLimitDegrees=15}
 {
  bone1Name=dyn_skirtBack_mstr_r
  bone2Name=pelvis
  swing1LimitDegrees=45
  swing1Motion=Free
  swing2LimitDegrees=45
  swing2Motion=Free
  twistLimitDegrees=45
  twistMotion=Free
 }
 {bone1Name=dyn_skirtBack_01_r, bone2Name=dyn_skirtBack_mstr_r, swing1LimitDegrees=45, swing2LimitDegrees=45, twistLimitDegrees=15}
 {bone1Name=dyn_skirtBack_02_r, bone2Name=dyn_skirtBack_01_r, swing1LimitDegrees=45, swing2LimitDegrees=45, twistLimitDegrees=15}
 {bone1Name=dyn_skirtBack_03_r, bone2Name=dyn_skirtBack_02_r, swing1LimitDegrees=45, swing2LimitDegrees=45, twistLimitDegrees=15}
 {bone1Name=dyn_skirtBack_04_l, bone2Name=dyn_skirtBack_03_l, swing1LimitDegrees=45, swing2LimitDegrees=45, twistLimitDegrees=15}
 {
  bone1Name=dyn_skirtBack_mstr_l
  bone2Name=pelvis
  swing1LimitDegrees=45
  swing1Motion=Free
  swing2LimitDegrees=45
  swing2Motion=Free
  twistLimitDegrees=45
  twistMotion=Free
 }
 {bone1Name=dyn_skirtBack_01_l, bone2Name=dyn_skirtBack_mstr_l, swing1LimitDegrees=45, swing2LimitDegrees=45, twistLimitDegrees=15}
 {bone1Name=dyn_skirtBack_02_l, bone2Name=dyn_skirtBack_01_l, swing1LimitDegrees=45, swing2LimitDegrees=45, twistLimitDegrees=15}
 {bone1Name=dyn_skirtBack_03_l, bone2Name=dyn_skirtBack_02_l, swing1LimitDegrees=45, swing2LimitDegrees=45, twistLimitDegrees=15}
 {
  bone1Name=dyn_skirtFront_03
  bone2Name=dyn_skirtFront_03_r
  swing1LimitDegrees=45
  swing1Motion=Free
  swing2LimitDegrees=45
  swing2Motion=Free
  twistLimitDegrees=45
  twistMotion=Free
 }
 {
  bone1Name=dyn_skirtFront_03_l
  bone2Name=dyn_skirtSide_03_l
  swing1LimitDegrees=45
  swing1Motion=Free
  swing2LimitDegrees=45
  swing2Motion=Free
  twistLimitDegrees=45
  twistMotion=Free
 }
 {
  bone1Name=spine_01
  bone2Name=pelvis
  swing1LimitDegrees=45
  swing1Motion=Free
  swing2LimitDegrees=45
  swing2Motion=Free
  twistLimitDegrees=45
  twistMotion=Free
 }
 {
  bone1Name=spine_02
  bone2Name=spine_01
  swing1LimitDegrees=45
  swing1Motion=Free
  swing2LimitDegrees=45
  swing2Motion=Free
  twistLimitDegrees=45
  twistMotion=Free
 }
 {
  bone1Name=dyn_belt_B_MSTR
  bone2Name=spine_02
  swing1LimitDegrees=45
  swing1Motion=Free
  swing2LimitDegrees=45
  swing2Motion=Free
  twistLimitDegrees=45
  twistMotion=Free
 }
 {bone1Name=dyn_belt_b_01, bone2Name=dyn_belt_B_MSTR, swing1LimitDegrees=45, swing2LimitDegrees=45, twistLimitDegrees=5}
 {bone1Name=dyn_belt_b_02, bone2Name=dyn_belt_b_01, swing1LimitDegrees=35, swing2LimitDegrees=15, twistLimitDegrees=5}
 {bone1Name=dyn_belt_b_03, bone2Name=dyn_belt_b_02, swing1LimitDegrees=45, swing2LimitDegrees=30, twistLimitDegrees=5}
 {bone1Name=dyn_belt_b_04, bone2Name=dyn_belt_b_03, swing1LimitDegrees=45, swing2LimitDegrees=45, twistLimitDegrees=5}
 {
  bone1Name=dyn_belt_a_MSTR
  bone2Name=spine_02
  swing1LimitDegrees=45
  swing1Motion=Free
  swing2LimitDegrees=45
  swing2Motion=Free
  twistLimitDegrees=45
  twistMotion=Free
 }
 {bone1Name=dyn_belt_a_02, bone2Name=dyn_belt_a_MSTR, swing1LimitDegrees=45, swing2LimitDegrees=45, twistLimitDegrees=5}
 {bone1Name=dyn_belt_a_03, bone2Name=dyn_belt_a_02, swing1LimitDegrees=35, swing2LimitDegrees=15, twistLimitDegrees=5}
 {bone1Name=dyn_belt_a_04, bone2Name=dyn_belt_a_03, swing1LimitDegrees=45, swing2LimitDegrees=30, twistLimitDegrees=5}
 {bone1Name=dyn_belt_a_05, bone2Name=dyn_belt_a_04, swing1LimitDegrees=45, swing2LimitDegrees=45, twistLimitDegrees=5}
 {
  bone1Name=upperarm_l
  bone2Name=spine_02
  swing1LimitDegrees=45
  swing1Motion=Free
  swing2LimitDegrees=45
  swing2Motion=Free
  twistLimitDegrees=45
  twistMotion=Free
 }
 {
  bone1Name=lowerarm_l
  bone2Name=upperarm_l
  swing1LimitDegrees=45
  swing1Motion=Free
  swing2LimitDegrees=45
  swing2Motion=Free
  twistLimitDegrees=45
  twistMotion=Free
 }
 {
  bone1Name=upperarm_r
  bone2Name=spine_02
  swing1LimitDegrees=45
  swing1Motion=Free
  swing2LimitDegrees=45
  swing2Motion=Free
  twistLimitDegrees=45
  twistMotion=Free
 }
 {
  bone1Name=lowerarm_r
  bone2Name=upperarm_r
  swing1LimitDegrees=45
  swing1Motion=Free
  swing2LimitDegrees=45
  swing2Motion=Free
  twistLimitDegrees=45
  twistMotion=Free
 }
]
