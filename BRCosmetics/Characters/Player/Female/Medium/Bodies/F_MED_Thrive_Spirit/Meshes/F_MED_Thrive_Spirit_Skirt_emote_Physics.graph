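# Physics asset for the Thrive Spirit female medium body's skirt, an emote variant by its name, defining collision and joints for skirt motion.
# Capsules cover pelvis and legs, with a sphere as floor collider.
# Front, side and back bone chains use limited joints, hanging from free-rotating masters on the pelvis, except front-left on the left thigh.
# Free-angular cross-constraints tie some neighboring chains together at middle segments, asymmetrically between sides.
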PhysicsAsset /BRCosmetics/Characters/Player/Female/Medium/Bodies/F_MED_Thrive_Spirit/Meshes/F_MED_Thrive_Spirit_Skirt_emote_Physics
bodies {
 calf_l [{shapeType=Capsule, length=44.3711, radius=9.5875, center=<-28.7096 -1.7372 0.0465>, rotation=<rot 0 89.533 -88.5703>}]
 calf_r [{shapeType=Capsule, length=40.4961, radius=13.4625, center=<28.7325 1.5217 1.473>, rotation=<rot -0.3385 89.533 -88.4634>}]
 calf_twist_02_l [{
   shapeType=Capsule
   length=24.9493
   radius=9.0287
   shapeName=calf_twist_02_l_capsule
   center=<-4.0161 -1.0162 -0.0758>
   rotation=<rot 0 -93.6169 -89.6979>
  }]
 dyn_floor_COL [{shapeType=Sphere, radius=39.005, shapeName=dyn_floor_COL_sphere, center=<-0.0004 -16.0885 -4.7633>}]
 dyn_skirtBack_01_l [{shapeType=Capsule, length=13.75, radius=7.25, shapeName=dyn_skirtBack_01_l_capsule, center=<5 0 6>, rotation=<rot 0 0 94.9998>}]
 dyn_skirtBack_01_r [{shapeType=Capsule, length=13.25, radius=7.75, shapeName=dyn_skirtBack_01_r_capsule, center=<5 0 6>, rotation=<rot 0 0 94.9999>}]
 dyn_skirtBack_02_l [{shapeType=Capsule, length=16.25, radius=9.5, shapeName=dyn_skirtBack_02_l_capsule, center=<5 3 6>, rotation=<rot 0 0 94.9998>}]
 dyn_skirtBack_02_r [{shapeType=Capsule, length=16, radius=9.75, shapeName=dyn_skirtBack_02_r_capsule, center=<5 -3 6>, rotation=<rot 0 0 89.9986>}]
 dyn_skirtBack_03_l [{shapeType=Capsule, length=23.3243, radius=10.75, shapeName=dyn_skirtBack_03_l_capsule, center=<5 3 6>, rotation=<rot 0 0 94.9999>}]
 dyn_skirtBack_03_r [{shapeType=Capsule, length=23.8243, radius=10.25, shapeName=dyn_skirtBack_03_r_capsule, center=<5 -3 6>, rotation=<rot 0 0 89.9986>}]
 dyn_skirtBack_04_l [{shapeType=Capsule, length=24.3243, radius=11.75, shapeName=dyn_skirtBack_04_l_capsule, center=<5 3 6>, rotation=<rot 0 0 94.9999>}]
 dyn_skirtBack_04_r [{shapeType=Capsule, length=25.0743, radius=11, shapeName=dyn_skirtBack_04_r_capsule, center=<5 -3 6>, rotation=<rot 0 0 89.999>}]
 dyn_skirtBack_mstr_l [{shapeType=Capsule, length=0.505, radius=0.505, shapeName=dyn_skirtBack_mstr_l_capsule, rotation=<rot 0 -90 0>}]
 dyn_skirtBack_mstr_r [{shapeType=Capsule, length=0.505, radius=0.505, shapeName=dyn_skirtBack_mstr_r_capsule, rotation=<rot 0 -90 0>}]
 dyn_skirtFront_01_l [{shapeType=Capsule, length=0.505, radius=0.505, shapeName=dyn_skirtFront_01_l_capsule, rotation=<rot 0 -90 0>}]
 dyn_skirtFront_01_r [{shapeType=Capsule, length=16.5, radius=6, shapeName=dyn_skirtFront_01_r_capsule, center=<6 0 6>, rotation=<rot 0 0 90>}]
 dyn_skirtFront_02_l [{shapeType=Capsule, length=13, radius=6, shapeName=dyn_skirtFront_02_l_capsule, center=<6 0 9>, rotation=<rot 0 0 109.9999>}]
 dyn_skirtFront_02_r [{shapeType=Capsule, length=17.75, radius=7, shapeName=dyn_skirtFront_02_r_capsule, center=<6 0 6>, rotation=<rot 0 0 85>}]
 dyn_skirtFront_03_l [{shapeType=Capsule, length=13, radius=6, shapeName=dyn_skirtFront_03_l_capsule, center=<6 0 8>, rotation=<rot 0 0 110.0005>}]
 dyn_skirtFront_03_r [{shapeType=Capsule, length=19.625, radius=6.125, shapeName=dyn_skirtFront_03_r_capsule, center=<6 0 6>, rotation=<rot 0 0 84.9998>}]
 dyn_skirtFront_04 [{shapeType=Capsule, length=12.8829, radius=7.8066, shapeName=dyn_skirtFront_04_capsule, center=<7 5 6>, rotation=<rot 0 0 90>}]
 dyn_skirtFront_04_l [{shapeType=Capsule, length=12.5, radius=6.5, shapeName=dyn_skirtFront_04_l_capsule, center=<6 0 7>, rotation=<rot 0 0 110.0005>}]
 dyn_skirtFront_04_r [{shapeType=Capsule, length=19.375, radius=7.375, shapeName=dyn_skirtFront_04_r_capsule, center=<6 0 6>, rotation=<rot 0 0 85.0002>}]
 dyn_skirtFront_mstr [{shapeType=Capsule, length=0.505, radius=0.505, shapeName=dyn_skirtFront_mstr_capsule}]
 dyn_skirtFront_mstr_l [{shapeType=Capsule, length=0.505, radius=0.505, shapeName=dyn_skirtFront_mstr_l_capsule, rotation=<rot 0 -90 0>}]
 dyn_skirtFront_mstr_r [{shapeType=Capsule, length=0.505, radius=0.505, shapeName=dyn_skirtFront_mstr_r_capsule, rotation=<rot 0 -90 0>}]
 dyn_skirtSide_01_l [{shapeType=Capsule, length=11.125, radius=7.625, shapeName=dyn_skirtSide_01_l_capsule, center=<5 -3 6>, rotation=<rot 0 0 65.8892>}]
 dyn_skirtSide_01_r [{shapeType=Capsule, length=17.375, radius=6.875, shapeName=dyn_skirtSide_01_r_capsule, center=<6 0 6>, rotation=<rot 0 0 90>}]
 dyn_skirtSide_02_l [{shapeType=Capsule, length=20.4019, radius=8.5, shapeName=dyn_skirtSide_02_l_capsule, center=<5 -3 6>, rotation=<rot 0 0 65.8878>}]
 dyn_skirtSide_02_r [{shapeType=Capsule, length=17.125, radius=7.125, shapeName=dyn_skirtSide_02_r_capsule, center=<6 0 6>, rotation=<rot 0 0 90>}]
 dyn_skirtSide_03_l [{shapeType=Capsule, length=25.466, radius=9.75, shapeName=dyn_skirtSide_03_l_capsule, center=<5 -3 6>, rotation=<rot 0 0 65.8887>}]
 dyn_skirtSide_03_r [{shapeType=Capsule, length=16.875, radius=8.375, shapeName=dyn_skirtSide_03_r_capsule, center=<6 0 6>, rotation=<rot 0 0 90>}]
 dyn_skirtSide_04_l [{shapeType=Capsule, length=25.966, radius=10.25, shapeName=dyn_skirtSide_04_l_capsule, center=<5 -3 6>, rotation=<rot 0 0 65.8885>}]
 dyn_skirtSide_04_r [{shapeType=Capsule, length=16.75, radius=9.5, shapeName=dyn_skirtSide_04_r_capsule, center=<6 0 6>, rotation=<rot 0 0 90>}]
 dyn_skirtSide_mstr_l [{shapeType=Capsule, length=0.505, radius=0.505, shapeName=dyn_skirtSide_mstr_l_capsule, rotation=<rot 0 -90 0>}]
 dyn_skirtSide_mstr_r [{shapeType=Capsule, length=0.505, radius=0.505, shapeName=dyn_skirtSide_mstr_r_capsule, rotation=<rot 0 -90 0>}]
 pelvis [
  {shapeType=Capsule, length=9.0875, radius=9.6801, center=<-0.9227 -1.2812 0>}
  {shapeType=Capsule, length=12.415, radius=7.9702, shapeName=pelvis_capsule, center=<-3.9978 -3.7131 0>}
 ]
 thigh_l [{shapeType=Capsule, length=27.2111, radius=8.7722, center=<-21.4716 0.4992 -0.2548>, rotation=<rot 90 -7.5103 -7.3149>}]
 thigh_r [{shapeType=Capsule, length=23.33, radius=12.6533, center=<21.7868 -0.0552 -1.1637>, rotation=<rot 88.343 -180 -180>}]
 thigh_twist_01_r [{shapeType=Capsule, length=13.4234, radius=7.1932, shapeName=thigh_twist_01_r_capsule, center=<6.6879 0.8718 -0.0003>, rotation=<rot 90 0 0>}]
}
constraints [
 {bone1Name=thigh_l, bone2Name=pelvis, swing1LimitDegrees=45, swing2LimitDegrees=45, twistLimitDegrees=45}
 {bone1Name=calf_l, bone2Name=thigh_l, swing1LimitDegrees=45, swing2LimitDegrees=45, twistLimitDegrees=45}
 {bone1Name=thigh_r, bone2Name=pelvis, swing1LimitDegrees=45, swing2LimitDegrees=45, twistLimitDegrees=45}
 {bone1Name=calf_r, bone2Name=thigh_r, swing1LimitDegrees=45, swing2LimitDegrees=45, twistLimitDegrees=45}
 {
  bone1Name=dyn_skirtFront_mstr
  bone2Name=pelvis
  swing1LimitDegrees=45
  swing1Motion=Free
  swing2LimitDegrees=45
  swing2Motion=Free
  twistLimitDegrees=45
  twistMotion=Free
 }
 {bone1Name=dyn_skirtSide_04_l, bone2Name=dyn_skirtSide_03_l, swing1LimitDegrees=35, swing2LimitDegrees=35, twistLimitDegrees=15}
 {
  bone1Name=dyn_skirtSide_mstr_l
  bone2Name=pelvis
  swing1LimitDegrees=45
  swing1Motion=Free
  swing2LimitDegrees=45
  swing2Motion=Free
  twistLimitDegrees=45
  twistMotion=Free
 }
 {bone1Name=dyn_skirtSide_01_l, bone2Name=dyn_skirtSide_mstr_l, swing1LimitDegrees=35, swing2LimitDegrees=35, twistLimitDegrees=15}
 {bone1Name=dyn_skirtSide_02_l, bone2Name=dyn_skirtSide_01_l, swing1LimitDegrees=35, swing2LimitDegrees=35, twistLimitDegrees=15}
 {bone1Name=dyn_skirtSide_03_l, bone2Name=dyn_skirtSide_02_l, swing1LimitDegrees=35, swing2LimitDegrees=35, twistLimitDegrees=15}
 {bone1Name=dyn_skirtBack_04_r, bone2Name=dyn_skirtBack_03_r, swing1LimitDegrees=35, swing2LimitDegrees=35, twistLimitDegrees=15}
 {
  bone1Name=dyn_skirtBack_mstr_r
  bone2Name=pelvis
  swing1LimitDegrees=45
  swing1Motion=Free
  swing2LimitDegrees=45
  swing2Motion=Free
  twistLimitDegrees=45
  twistMotion=Free
 }
 {bone1Name=dyn_skirtBack_01_r, bone2Name=dyn_skirtBack_mstr_r, swing1LimitDegrees=35, swing2LimitDegrees=35, twistLimitDegrees=15}
 {bone1Name=dyn_skirtBack_02_r, bone2Name=dyn_skirtBack_01_r, swing1LimitDegrees=35, swing2LimitDegrees=35, twistLimitDegrees=15}
 {bone1Name=dyn_skirtBack_03_r, bone2Name=dyn_skirtBack_02_r, swing1LimitDegrees=35, swing2LimitDegrees=35, twistLimitDegrees=15}
 {bone1Name=dyn_skirtBack_04_l, bone2Name=dyn_skirtBack_03_l, swing1LimitDegrees=35, swing2LimitDegrees=35, twistLimitDegrees=15}
 {
  bone1Name=dyn_skirtBack_mstr_l
  bone2Name=pelvis
  swing1LimitDegrees=45
  swing1Motion=Free
  swing2LimitDegrees=45
  swing2Motion=Free
  twistLimitDegrees=45
  twistMotion=Free
 }
 {bone1Name=dyn_skirtBack_01_l, bone2Name=dyn_skirtBack_mstr_l, swing1LimitDegrees=35, swing2LimitDegrees=35, twistLimitDegrees=15}
 {bone1Name=dyn_skirtBack_02_l, bone2Name=dyn_skirtBack_01_l, swing1LimitDegrees=35, swing2LimitDegrees=35, twistLimitDegrees=15}
 {bone1Name=dyn_skirtBack_03_l, bone2Name=dyn_skirtBack_02_l, swing1LimitDegrees=35, swing2LimitDegrees=35, twistLimitDegrees=15}
 {bone1Name=dyn_skirtSide_04_r, bone2Name=dyn_skirtSide_03_r, swing1LimitDegrees=35, swing2LimitDegrees=35, twistLimitDegrees=15}
 {
  bone1Name=dyn_skirtSide_mstr_r
  bone2Name=pelvis
  swing1LimitDegrees=45
  swing1Motion=Free
  swing2LimitDegrees=45
  swing2Motion=Free
  twistLimitDegrees=45
  twistMotion=Free
 }
 {bone1Name=dyn_skirtSide_01_r, bone2Name=dyn_skirtSide_mstr_r, swing1LimitDegrees=35, swing2LimitDegrees=35, twistLimitDegrees=15}
 {bone1Name=dyn_skirtSide_02_r, bone2Name=dyn_skirtSide_01_r, swing1LimitDegrees=35, swing2LimitDegrees=35, twistLimitDegrees=15}
 {bone1Name=dyn_skirtSide_03_r, bone2Name=dyn_skirtSide_02_r, swing1LimitDegrees=35, swing2LimitDegrees=35, twistLimitDegrees=15}
 {bone1Name=dyn_skirtFront_04_r, bone2Name=dyn_skirtFront_03_r, swing1LimitDegrees=35, swing2LimitDegrees=35, twistLimitDegrees=15}
 {bone1Name=dyn_skirtFront_01_r, bone2Name=dyn_skirtFront_mstr_r, swing1LimitDegrees=5, swing2LimitDegrees=35, twistLimitDegrees=15}
 {bone1Name=dyn_skirtFront_02_r, bone2Name=dyn_skirtFront_01_r, swing1LimitDegrees=5, swing2LimitDegrees=35, twistLimitDegrees=15}
 {bone1Name=dyn_skirtFront_03_r, bone2Name=dyn_skirtFront_02_r, swing1LimitDegrees=5, swing2LimitDegrees=35, twistLimitDegrees=15}
 {
  bone1Name=dyn_skirtFront_mstr_r
  bone2Name=pelvis
  swing1LimitDegrees=45
  swing1Motion=Free
  swing2LimitDegrees=45
  swing2Motion=Free
  twistLimitDegrees=45
  twistMotion=Free
 }
 {
  bone1Name=dyn_skirtFront_mstr_l
  bone2Name=thigh_l
  swing1LimitDegrees=45
  swing1Motion=Free
  swing2LimitDegrees=45
  swing2Motion=Free
  twistLimitDegrees=45
  twistMotion=Free
 }
 {
  bone1Name=dyn_skirtFront_01_l
  bone2Name=dyn_skirtFront_mstr_l
  swing1LimitDegrees=45
  swing1Motion=Free
  swing2LimitDegrees=45
  swing2Motion=Free
  twistLimitDegrees=45
  twistMotion=Free
 }
 {bone1Name=dyn_skirtFront_02_l, bone2Name=dyn_skirtFront_01_l, swing1LimitDegrees=45, swing2LimitDegrees=45, twistLimitDegrees=15}
 {bone1Name=dyn_skirtFront_03_l, bone2Name=dyn_skirtFront_02_l, swing1LimitDegrees=45, swing2LimitDegrees=45, twistLimitDegrees=15}
 {bone1Name=dyn_skirtFront_04_l, bone2Name=dyn_skirtFront_03_l, swing1LimitDegrees=45, swing2LimitDegrees=45, twistLimitDegrees=15}
 {bone1Name=dyn_skirtFront_04, bone2Name=dyn_skirtFront_mstr, swing1LimitDegrees=45, swing2LimitDegrees=45, twistLimitDegrees=45}
 {
  bone1Name=thigh_twist_01_r
  bone2Name=thigh_r
  swing1LimitDegrees=45
  swing1Motion=Free
  swing2LimitDegrees=45
  swing2Motion=Free
  twistLimitDegrees=45
  twistMotion=Free
 }
 {
  bone1Name=calf_twist_02_l
  bone2Name=calf_l
  swing1LimitDegrees=45
  swing1Motion=Free
  swing2LimitDegrees=45
  swing2Motion=Free
  twistLimitDegrees=45
  twistMotion=Free
 }
 {
  bone1Name=dyn_skirtBack_03_l
  bone2Name=dyn_skirtBack_03_r
  swing1LimitDegrees=45
  swing1Motion=Free
  swing2LimitDegrees=45
  swing2Motion=Free
  twistLimitDegrees=45
  twistMotion=Free
 }
 {
  bone1Name=dyn_skirtBack_02_l
  bone2Name=dyn_skirtBack_02_r
  swing1LimitDegrees=45
  swing1Motion=Free
  swing2LimitDegrees=45
  swing2Motion=Free
  twistLimitDegrees=45
  twistMotion=Free
 }
 {
  bone1Name=dyn_skirtSide_02_r
  bone2Name=dyn_skirtFront_02_r
  swing1LimitDegrees=45
  swing1Motion=Free
  swing2LimitDegrees=45
  swing2Motion=Free
  twistLimitDegrees=45
  twistMotion=Free
 }
 {
  bone1Name=dyn_skirtSide_03_r
  bone2Name=dyn_skirtFront_03_r
  swing1LimitDegrees=45
  swing1Motion=Free
  swing2LimitDegrees=45
  swing2Motion=Free
  twistLimitDegrees=45
  twistMotion=Free
 }
 {
  bone1Name=dyn_skirtSide_02_l
  bone2Name=dyn_skirtFront_02_l
  swing1LimitDegrees=45
  swing1Motion=Free
  swing2LimitDegrees=45
  swing2Motion=Free
  twistLimitDegrees=45
  twistMotion=Free
 }
 {
  bone1Name=dyn_skirtSide_03_l
  bone2Name=dyn_skirtFront_03_l
  swing1LimitDegrees=45
  swing1Motion=Free
  swing2LimitDegrees=45
  swing2Motion=Free
  twistLimitDegrees=45
  twistMotion=Free
 }
 {
  bone1Name=dyn_skirtBack_03_l
  bone2Name=dyn_skirtSide_03_l
  swing1LimitDegrees=45
  swing1Motion=Free
  swing2LimitDegrees=45
  swing2Motion=Free
  twistLimitDegrees=45
  twistMotion=Free
 }
]
