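# Post-process anim Blueprint for the female medium Thrive Spirit body cosmetic, adding secondary motion to skirt, belts, ropes, necklace and bands.
# It copies the parent mesh pose, applies state-driven corrective bone offsets and constraints, then runs a Control Rig.
# RigidBody branches for skirt and ropes switch physics assets or simulation base bone by emote and skydive state.
# AnimDynamics springs finish the arm bands, cuffs and necklace, alongside sine-driven necklace rotations.
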AnimBlueprint /BRCosmetics/Characters/Player/Female/Medium/Bodies/F_MED_Thrive_Spirit/Meshes/F_MED_Thrive_Spirit_AnimBP
target_skeleton=&/F_MED_Thrive_Spirit_Skeleton
graphs [
 AnimGraph {nodes [
   n1 = RigidBody!(
    ComponentPose=Local To Component(LocalPose=Use cached pose 'prerigid'())
    Alpha
    OverrideWorldGravity:Vector=Get GravityOverride()
    ExternalForce
   ) -> (Pose) {
    bClampLinearTranslationLimitToRefPose=true
    bOverrideWorldGravity=true
    overridePhysicsAsset=&/F_MED_Thrive_Spirit_Skirt_Physics
    simulationSpace=BaseBoneSpace
    alphaBoolBlend {blendOption=Linear}
    baseBoneRef {boneName=dyn_root_simspace}
    componentAppliedLinearAccClamp=<1500 1500 1500>
    componentLinearAccScale=<0.25 0.25 0.4>
    componentLinearVelScale=<0.2 0.2 0.2>
   }
   n2 = RigidBody(ComponentPose=@n1, OverrideWorldGravity=Get GravityOverride_ropes() -> (GravityOverride_ropes:Vector)) {
    bClampLinearTranslationLimitToRefPose=true
    bOverrideWorldGravity=true
    overridePhysicsAsset=&/F_MED_Thrive_Spirit_Rope_Emote_Physics
    simulationSpace=BaseBoneSpace
    alphaBoolBlend {blendOption=Linear}
    baseBoneRef {boneName=dyn_root_simspace}
    componentAppliedLinearAccClamp=<3000 3000 3000>
    componentLinearAccScale=<0.3 0.3 0.3>
    componentLinearVelScale=<0.3 0.3 0.3>
    overrideWorldGravity {z=-980}
    simSpaceSettings {
     maxAngularAcceleration=1000
     maxAngularVelocity=3
     maxLinearAcceleration=700
     maxLinearVelocity=500
     velocityScaleZ=0
     worldAlpha=1
     externalLinearDragV=<1 1>
    }
   }
   Blend Poses by bool(BlendPose_0=Component To Local(ComponentPose=@n2), bActiveValue=Get bIsPlayingEmote()) {blendType=Linear}
   "Constraint - dyn_skirtFront_mstr (P)"(Alpha=1, ConstraintWeights_0=1) {
    alphaBoolBlend {blendOption=Linear}
    boneToModify {boneName=dyn_skirtFront_mstr}
    constraintSetup [{targetBone {boneName=thigh_twist_01_r}}]
    constraintWeights [1]
   }
   n3 = Get sineRotation3() -> (sineRotation3:Rotator)
   n4 = "Transform (Modify) Bone - Bone: dyn_necklace_01"(ComponentPose=Local To Component(LocalPose=Use cached pose 'post rigid'()), Alpha=1, Translation="0,0,0", Rotation=@n3, Scale="1,1,1") {rotationMode=BMM_Replace, rotationSpace=BCS_BoneSpace, alphaBoolBlend {blendOption=Linear}, boneToModify {boneName=dyn_necklace_01}}
   n5 = "Transform (Modify) Bone - Bone: dyn_necklace_09"(ComponentPose=@n4, Alpha=1, Translation="0,0,0", Rotation=@n3, Scale="1,1,1") {rotationMode=BMM_Replace, rotationSpace=BCS_BoneSpace, alphaBoolBlend {blendOption=Linear}, boneToModify {boneName=dyn_necklace_09}}
   n6 = "Transform (Modify) Bone - Bone: dyn_necklace_02"(ComponentPose=@n5, Alpha=1, Translation="0,0,0", Rotation=@n3, Scale="1,1,1") {rotationMode=BMM_Replace, rotationSpace=BCS_BoneSpace, alphaBoolBlend {blendOption=Linear}, boneToModify {boneName=dyn_necklace_02}}
   n7 = Get sineRotation4() -> (sineRotation4:Rotator)
   n8 = "Transform (Modify) Bone - Bone: dyn_necklace_03"(ComponentPose=@n6, Alpha=1, Translation="0,0,0", Rotation=@n7, Scale="1,1,1") {rotationMode=BMM_Replace, rotationSpace=BCS_BoneSpace, alphaBoolBlend {blendOption=Linear}, boneToModify {boneName=dyn_necklace_03}}
   n9 = Get sineRotation5() -> (sineRotation5:Rotator)
   n10 = "Transform (Modify) Bone - Bone: dyn_necklace_04"(ComponentPose=@n8, Alpha=1, Translation="0,0,0", Rotation=@n9, Scale="1,1,1") {rotationMode=BMM_Replace, rotationSpace=BCS_BoneSpace, alphaBoolBlend {blendOption=Linear}, boneToModify {boneName=dyn_necklace_04}}
   n11 = "Transform (Modify) Bone - Bone: dyn_necklace_08"(ComponentPose=@n10, Alpha=1, Translation="0,0,0", Rotation=@n3, Scale="1,1,1") {rotationMode=BMM_Replace, rotationSpace=BCS_BoneSpace, alphaBoolBlend {blendOption=Linear}, boneToModify {boneName=dyn_necklace_08}}
   n12 = "Transform (Modify) Bone - Bone: dyn_necklace_07"(ComponentPose=@n11, Alpha=1, Translation="0,0,0", Rotation=@n7, Scale="1,1,1") {rotationMode=BMM_Replace, rotationSpace=BCS_BoneSpace, alphaBoolBlend {blendOption=Linear}, boneToModify {boneName=dyn_necklace_07}}
   n13 = "Transform (Modify) Bone - Bone: dyn_necklace_06"(ComponentPose=@n12, Alpha=1, Translation="0,0,0", Rotation=@n9, Scale="1,1,1") {rotationMode=BMM_Replace, rotationSpace=BCS_BoneSpace, alphaBoolBlend {blendOption=Linear}, boneToModify {boneName=dyn_necklace_06}}
   n14 = Get rifle() -> (rifle:bool)
   n15 = "Transform (Modify) Bone - Bone: dyn_necklace_offset_02"!(ComponentPose=@n13, bAlphaBoolEnabled:bool=@n14, Translation="-0.000,6.000,-9.435", Rotation="0.000,-35.000,0", Scale="0.8,0.8,0.8") -> (Pose) {
    alphaInputType=Bool
    rotationMode=BMM_Additive
    rotationSpace=BCS_WorldSpace
    scaleMode=BMM_Additive
    scaleSpace=BCS_BoneSpace
    translationMode=BMM_Additive
    translationSpace=BCS_BoneSpace
    alphaBoolBlend {blendInTime=0.2, blendOption=Linear, blendOutTime=0.2}
    alphaScaleBiasClamp {interpSpeedDecreasing=2, interpSpeedIncreasing=2}
    boneToModify {boneName=dyn_necklace_offset_02}
    rotation {yaw=-35}
    scale=<0.8 0.8 0.8>
    translation {y=6, z=-9.435}
   }
   n16 = "Transform (Modify) Bone - Bone: dyn_necklace_offset_03"!(ComponentPose=@n15, bAlphaBoolEnabled:bool=@n14, Translation="-2,3,-12", Rotation="0,0,0", Scale="0.8,0.8,0.8") -> (Pose) {
    alphaInputType=Bool
    rotationMode=BMM_Additive
    rotationSpace=BCS_WorldSpace
    scaleMode=BMM_Additive
    scaleSpace=BCS_BoneSpace
    translationMode=BMM_Additive
    translationSpace=BCS_BoneSpace
    alphaBoolBlend {blendInTime=0.2, blendOption=Linear, blendOutTime=0.2}
    alphaScaleBiasClamp {interpSpeedDecreasing=2, interpSpeedIncreasing=2}
    boneToModify {boneName=dyn_necklace_offset_03}
    scale=<0.8 0.8 0.8>
    translation=<-2 3 -12>
   }
   n17 = "Transform (Modify) Bone - Bone: dyn_necklace_offset_02"!(ComponentPose=@n16, Alpha:float=Get Necklace 01() -> (Necklace 01:double), Translation="1,0,2.5", Rotation="0,0,0", Scale="1,1,1") -> (Pose) {
    rotationMode=BMM_Additive
    rotationSpace=BCS_WorldSpace
    scaleMode=BMM_Additive
    scaleSpace=BCS_BoneSpace
    translationMode=BMM_Additive
    translationSpace=BCS_BoneSpace
    alphaBoolBlend {blendOption=Linear}
    alphaScaleBiasClamp {interpSpeedDecreasing=2, interpSpeedIncreasing=2}
    boneToModify {boneName=dyn_necklace_offset_02}
    translation {x=1, z=2.5}
   }
   n18 = "Transform (Modify) Bone - Bone: dyn_necklace_offset_02"(
    ComponentPose=@n17
    Alpha=Get Head Rot_2() -> (Head Rot_2:double)
    Translation="-0.000,6.000,-9.435"
    Rotation="0.000,-35.000,0"
    Scale="0.8,0.8,0.8"
   ) {
    rotationMode=BMM_Additive
    rotationSpace=BCS_WorldSpace
    scaleMode=BMM_Additive
    scaleSpace=BCS_BoneSpace
    translationMode=BMM_Additive
    translationSpace=BCS_BoneSpace
    alphaBoolBlend {blendInTime=0.2, blendOption=Linear, blendOutTime=0.2}
    alphaScaleBiasClamp {interpSpeedDecreasing=2, interpSpeedIncreasing=2}
    boneToModify {boneName=dyn_necklace_offset_02}
    rotation {yaw=-35}
    scale=<0.8 0.8 0.8>
    translation {y=6, z=-9.435}
   }
   n19 = "Transform (Modify) Bone - Bone: dyn_necklace_offset_04"!(ComponentPose=@n18, bAlphaBoolEnabled:bool=True, Translation="3.000,0.000,2", Rotation="0,0,0", Scale="1,1,1") -> (Pose) {
    alphaInputType=Bool
    bAlphaBoolEnabled=false
    rotationMode=BMM_Additive
    rotationSpace=BCS_WorldSpace
    scaleMode=BMM_Additive
    scaleSpace=BCS_BoneSpace
    translationMode=BMM_Additive
    translationSpace=BCS_BoneSpace
    alphaBoolBlend {blendInTime=0.2, blendOption=Linear, blendOutTime=0.2}
    alphaScaleBiasClamp {interpSpeedDecreasing=2, interpSpeedIncreasing=2}
    boneToModify {boneName=dyn_necklace_offset_04}
    translation {x=3, z=2}
   }
   n20 = "Transform (Modify) Bone - Bone: dyn_necklace_offset_06"!(ComponentPose=@n19, bAlphaBoolEnabled:bool=True, Translation="3.000,0.000,-1", Rotation="0,0,0", Scale="1,1,1") -> (Pose) {
    alphaInputType=Bool
    bAlphaBoolEnabled=false
    rotationMode=BMM_Additive
    rotationSpace=BCS_WorldSpace
    scaleMode=BMM_Additive
    scaleSpace=BCS_BoneSpace
    translationMode=BMM_Additive
    translationSpace=BCS_BoneSpace
    alphaBoolBlend {blendInTime=0.2, blendOption=Linear, blendOutTime=0.2}
    alphaScaleBiasClamp {interpSpeedDecreasing=2, interpSpeedIncreasing=2}
    boneToModify {boneName=dyn_necklace_offset_06}
    translation {x=3, z=-1}
   }
   n21 = "Transform (Modify) Bone - Bone: dyn_necklace_offset_05"!(ComponentPose=@n20, bAlphaBoolEnabled:bool=True, Translation="6,6,-0.000", Rotation="0,0,0", Scale="1,1,1") -> (Pose) {
    alphaInputType=Bool
    bAlphaBoolEnabled=false
    rotationMode=BMM_Additive
    rotationSpace=BCS_WorldSpace
    scaleMode=BMM_Additive
    scaleSpace=BCS_BoneSpace
    translationMode=BMM_Additive
    translationSpace=BCS_BoneSpace
    alphaBoolBlend {blendInTime=0.2, blendOption=Linear, blendOutTime=0.2}
    alphaScaleBiasClamp {interpSpeedDecreasing=2, interpSpeedIncreasing=2}
    boneToModify {boneName=dyn_necklace_offset_05}
    translation=<6 6>
   }
   n22 = "Constraint - dyn_necklace_05 (PP)"!(ComponentPose=@n21, Alpha=1, ConstraintWeights_0=0.5, ConstraintWeights_1:float=0.5) -> (Pose) {
    alphaBoolBlend {blendOption=Linear}
    boneToModify {boneName=dyn_necklace_05}
    constraintSetup [{targetBone {boneName=dyn_necklace_04}}, {targetBone {boneName=dyn_necklace_06}}]
    constraintWeights [1,1]
   }
   n23 = Anim Dynamics - Bone: dyn_necklace_offset_04(ComponentPose=@n22, Alpha=0.15, ExternalForce="0,0,0") {
    angularDampingOverride=1
    angularSpringConstant=1000
    bAngularSpring=true
    bLinearSpring=true
    bOverrideAngularDamping=true
    gravityScale=0.15
    linearSpringConstant=25
    simulationSpace=RootRelative
    alphaBoolBlend {blendOption=Linear}
    boundBone {boneName=dyn_necklace_offset_04}
    componentLinearAccScale=<0.65 0.65 0.65>
    componentLinearVelScale=<0.65 0.65 0.65>
    gravityOverride {z=-980}
    physicsBodyDefinitions [{
      sphereCollisionRadius=0
      boundBone {boneName=dyn_necklace_offset_04}
      boxExtents=<20 30 30>
      constraintSetup {angularLimitsMax=<15 15 15>, angularLimitsMin=<-15 -15 -15>, angularTarget=<1 0 0>, linearAxesMax=<5 1 2>, linearAxesMin=<0 0 -2>}
      localJointOffset=<-1 0 0>
     }]
   }
   n24 = Anim Dynamics - Bone: dyn_necklace_offset_05(ComponentPose=@n23, Alpha=0.25, ExternalForce="0,0,0") {
    angularDampingOverride=1
    angularSpringConstant=1000
    bAngularSpring=true
    bLinearSpring=true
    bOverrideAngularDamping=true
    gravityScale=0.15
    linearSpringConstant=25
    simulationSpace=RootRelative
    alphaBoolBlend {blendOption=Linear}
    boundBone {boneName=dyn_necklace_offset_05}
    componentLinearAccScale=<0.65 0.65 0.65>
    componentLinearVelScale=<0.65 0.65 0.65>
    gravityOverride {z=-980}
    physicsBodyDefinitions [{
      sphereCollisionRadius=0
      boundBone {boneName=dyn_necklace_offset_05}
      boxExtents=<20 30 30>
      constraintSetup {angularLimitsMax=<15 15 15>, angularLimitsMin=<-15 -15 -15>, angularTarget=<1 0 0>, linearAxesMax=<10 1 3>, linearAxesMin=<0 0 -3>}
      localJointOffset=<-1 0 0>
     }]
   }
   n25 = Anim Dynamics - Bone: dyn_necklace_offset_06(ComponentPose=@n24, Alpha=0.15, ExternalForce="0,0,0") {
    angularDampingOverride=1
    angularSpringConstant=1000
    bAngularSpring=true
    bLinearSpring=true
    bOverrideAngularDamping=true
    gravityScale=0.15
    linearSpringConstant=25
    simulationSpace=RootRelative
    alphaBoolBlend {blendOption=Linear}
    boundBone {boneName=dyn_necklace_offset_06}
    componentLinearAccScale=<0.65 0.65 0.65>
    componentLinearVelScale=<0.65 0.65 0.65>
    gravityOverride {z=-980}
    physicsBodyDefinitions [{
      sphereCollisionRadius=0
      boundBone {boneName=dyn_necklace_offset_06}
      boxExtents=<20 30 30>
      constraintSetup {angularLimitsMax=<15 15 15>, angularLimitsMin=<-15 -15 -15>, angularTarget=<1 0 0>, linearAxesMax=<5 1 2>, linearAxesMin=<0 0 -2>}
      localJointOffset=<-1 0 0>
     }]
   }
   Output Pose(Result=Component To Local(ComponentPose=@n25)) {layerGroup=None}
   n26 = "Transform (Modify) Bone - Bone: dyn_skirtFront_03_l"(
    ComponentPose=Local To Component(LocalPose=Use cached pose 'constraints'())
    Alpha=Get thigh_rot_l()
    Translation="0,0,0"
    Rotation="-13.434,-35,9.820"
    Scale="1,1,1"
   ) {
    rotationMode=BMM_Additive
    rotationSpace=BCS_BoneSpace
    scaleMode=BMM_Additive
    scaleSpace=BCS_BoneSpace
    translationMode=BMM_Additive
    translationSpace=BCS_BoneSpace
    alphaBoolBlend {blendOption=Linear}
    boneToModify {boneName=dyn_skirtFront_03_l}
    rotation=<rot -13.434 -35 9.82>
   }
   n27 = "Transform (Modify) Bone - Bone: dyn_skirtBack_mstr_l"(ComponentPose=@n26, Alpha=Get thigh_rot_l_back() -> (thigh_rot_l_back:double), Translation="0,0,0", Rotation="30,0,0", Scale="1,1,1") {
    rotationMode=BMM_Additive
    rotationSpace=BCS_BoneSpace
    scaleMode=BMM_Additive
    scaleSpace=BCS_BoneSpace
    translationMode=BMM_Additive
    translationSpace=BCS_BoneSpace
    alphaBoolBlend {blendOption=Linear}
    boneToModify {boneName=dyn_skirtBack_mstr_l}
    rotation {pitch=30}
   }
   n28 = "Transform (Modify) Bone - Bone: dyn_skirtBack_mstr_r"(ComponentPose=@n27, Alpha=Get thigh_rot_r_back() -> (thigh_rot_r_back:double), Translation="0,0,0", Rotation="30,0,0", Scale="1,1,1") {
    rotationMode=BMM_Additive
    rotationSpace=BCS_BoneSpace
    scaleMode=BMM_Additive
    scaleSpace=BCS_BoneSpace
    translationMode=BMM_Additive
    translationSpace=BCS_BoneSpace
    alphaBoolBlend {blendOption=Linear}
    boneToModify {boneName=dyn_skirtBack_mstr_r}
    rotation {pitch=30}
   }
   n29 = NOT Boolean(A=Get bIsSliding())
   n30 = "Constraint - dyn_skirtSide_mstr_l (RR)"!(ComponentPose=@n28, bAlphaBoolEnabled:bool=@n29, ConstraintWeights_0=0.5, ConstraintWeights_1:float=0.5) -> (Pose) {
    alphaInputType=Bool
    alphaBoolBlend {blendInTime=0.1, blendOption=Linear, blendOutTime=0.1}
    boneToModify {boneName=dyn_skirtSide_mstr_l}
    constraintSetup [{transformType=Rotation, targetBone {boneName=thigh_l}}, {transformType=Rotation, targetBone {boneName=dyn_skirtBack_mstr_l}}]
    constraintWeights [1,1]
   }
   n31 = "Constraint - dyn_skirtSide_mstr_r (RR)"!(ComponentPose=@n30, bAlphaBoolEnabled:bool=@n29, ConstraintWeights_0=0.5, ConstraintWeights_1:float=0.5) -> (Pose) {
    alphaInputType=Bool
    alphaBoolBlend {blendInTime=0.1, blendOption=Linear, blendOutTime=0.1}
    boneToModify {boneName=dyn_skirtSide_mstr_r}
    constraintSetup [{transformType=Rotation, targetBone {boneName=thigh_r}}, {transformType=Rotation, targetBone {boneName=dyn_skirtBack_mstr_r}}]
    constraintWeights [1,1]
   }
   n32 = "Transform (Modify) Bone - Bone: dyn_skirtSide_mstr_l"!(ComponentPose=@n31, bAlphaBoolEnabled:bool=True, Translation="0,0,0", Rotation="10,0.000,-0.000", Scale="1,1,1") -> (Pose) {
    alphaInputType=Bool
    rotationMode=BMM_Additive
    rotationSpace=BCS_BoneSpace
    scaleMode=BMM_Additive
    scaleSpace=BCS_BoneSpace
    translationMode=BMM_Additive
    translationSpace=BCS_BoneSpace
    alphaBoolBlend {blendInTime=0.1, blendOption=Linear, blendOutTime=0.1}
    boneToModify {boneName=dyn_skirtSide_mstr_l}
    rotation {pitch=10}
   }
   n33 = "Transform (Modify) Bone - Bone: dyn_skirtSide_mstr_r"!(ComponentPose=@n32, bAlphaBoolEnabled:bool=True, Translation="0,0,0", Rotation="10,0,0", Scale="1,1,1") -> (Pose) {
    alphaInputType=Bool
    rotationMode=BMM_Additive
    rotationSpace=BCS_BoneSpace
    scaleMode=BMM_Additive
    scaleSpace=BCS_BoneSpace
    translationMode=BMM_Additive
    translationSpace=BCS_BoneSpace
    alphaBoolBlend {blendInTime=0.1, blendOption=Linear, blendOutTime=0.1}
    boneToModify {boneName=dyn_skirtSide_mstr_r}
    rotation {pitch=10}
   }
   n34 = Get JumpingAlpha() -> (JumpingAlpha:double)
   n35 = "Transform (Modify) Bone - Bone: dyn_skirtBack_mstr_l"(ComponentPose=@n33, Alpha=@n34, Translation="0,0,0", Rotation="35,10,0.000", Scale="1,1,1") {
    rotationMode=BMM_Additive
    rotationSpace=BCS_BoneSpace
    alphaBoolBlend {blendOption=Linear}
    alphaScaleBiasClamp {bInterpResult=true, interpSpeedDecreasing=2, interpSpeedIncreasing=5}
    boneToModify {boneName=dyn_skirtBack_mstr_l}
    rotation {pitch=35, yaw=10}
   }
   n36 = "Transform (Modify) Bone - Bone: dyn_skirtBack_mstr_r"(ComponentPose=@n35, Alpha=@n34, Translation="0,0,0", Rotation="35,-10,0.000", Scale="1,1,1") {
    rotationMode=BMM_Additive
    rotationSpace=BCS_BoneSpace
    alphaBoolBlend {blendOption=Linear}
    alphaScaleBiasClamp {bInterpResult=true, interpSpeedDecreasing=2, interpSpeedIncreasing=5}
    boneToModify {boneName=dyn_skirtBack_mstr_r}
    rotation {pitch=35, yaw=-10}
   }
   n37 = "Transform (Modify) Bone - Bone: dyn_skirtSide_01_l"(ComponentPose=@n36, Alpha=@n34, Translation="0,0,0", Rotation="15,-15,0.000", Scale="1,1,1") {
    rotationMode=BMM_Additive
    rotationSpace=BCS_BoneSpace
    alphaBoolBlend {blendOption=Linear}
    alphaScaleBiasClamp {bInterpResult=true, interpSpeedDecreasing=2, interpSpeedIncreasing=5}
    boneToModify {boneName=dyn_skirtSide_01_l}
    rotation {pitch=15, yaw=-15}
   }
   n38 = "Transform (Modify) Bone - Bone: dyn_skirtSide_01_r"(ComponentPose=@n37, Alpha=@n34, Translation="0,0,0", Rotation="15,15,0.000", Scale="1,1,1") {
    rotationMode=BMM_Additive
    rotationSpace=BCS_BoneSpace
    alphaBoolBlend {blendOption=Linear}
    alphaScaleBiasClamp {bInterpResult=true, interpSpeedDecreasing=2, interpSpeedIncreasing=5}
    boneToModify {boneName=dyn_skirtSide_01_r}
    rotation {pitch=15, yaw=15}
   }
   n39 = Get SkydiveAlpha() -> (SkydiveAlpha:double)
   n40 = "Transform (Modify) Bone - Bone: dyn_skirtBack_mstr_l"(ComponentPose=@n38, Alpha=@n39, Translation="0,0,0", Rotation="0,0,0", Scale="1,1,1") {
    rotationMode=BMM_Additive
    rotationSpace=BCS_BoneSpace
    alphaBoolBlend {blendOption=Linear}
    alphaScaleBiasClamp {bInterpResult=true, interpSpeedDecreasing=2, interpSpeedIncreasing=5}
    boneToModify {boneName=dyn_skirtBack_mstr_l}
   }
   n41 = "Transform (Modify) Bone - Bone: dyn_skirtBack_03_l"(ComponentPose=@n40, Alpha=@n39, Translation="0,0,0", Rotation="20,0,0.000", Scale="1,1,1") {
    rotationMode=BMM_Additive
    rotationSpace=BCS_BoneSpace
    alphaBoolBlend {blendOption=Linear}
    alphaScaleBiasClamp {bInterpResult=true, interpSpeedDecreasing=2, interpSpeedIncreasing=5}
    boneToModify {boneName=dyn_skirtBack_03_l}
    rotation {pitch=20}
   }
   n42 = "Transform (Modify) Bone - Bone: dyn_skirtBack_mstr_r"(ComponentPose=@n41, Alpha=@n39, Translation="0,0,0", Rotation="0,0,0", Scale="1,1,1") {
    rotationMode=BMM_Additive
    rotationSpace=BCS_BoneSpace
    alphaBoolBlend {blendOption=Linear}
    alphaScaleBiasClamp {bInterpResult=true, interpSpeedDecreasing=2, interpSpeedIncreasing=5}
    boneToModify {boneName=dyn_skirtBack_mstr_r}
   }
   n43 = "Transform (Modify) Bone - Bone: dyn_belt_B_MSTR"(ComponentPose=@n42, Alpha=@n39, Translation="0,0,0", Rotation="0,15,0", Scale="1,1,1") {
    rotationMode=BMM_Additive
    rotationSpace=BCS_BoneSpace
    alphaBoolBlend {blendOption=Linear}
    alphaScaleBiasClamp {bInterpResult=true, interpSpeedDecreasing=2, interpSpeedIncreasing=5}
    boneToModify {boneName=dyn_belt_B_MSTR}
    rotation {yaw=15}
   }
   n44 = "Transform (Modify) Bone - Bone: dyn_belt_a_MSTR"(ComponentPose=@n43, Alpha=@n39, Translation="0,0,0", Rotation="0,5,0", Scale="1,1,1") {
    rotationMode=BMM_Additive
    rotationSpace=BCS_BoneSpace
    alphaBoolBlend {blendOption=Linear}
    alphaScaleBiasClamp {bInterpResult=true, interpSpeedDecreasing=2, interpSpeedIncreasing=5}
    boneToModify {boneName=dyn_belt_a_MSTR}
    rotation {yaw=5}
   }
   n45 = "Transform (Modify) Bone - Bone: dyn_skirtSide_mstr_r"!(ComponentPose=@n44, Alpha:float=@n39, Translation="0,0,0", Rotation="15,0,0.000", Scale="1,1,1") -> (Pose) {
    rotationMode=BMM_Additive
    rotationSpace=BCS_BoneSpace
    alphaBoolBlend {blendOption=Linear}
    alphaScaleBiasClamp {bInterpResult=true, interpSpeedDecreasing=2, interpSpeedIncreasing=5}
    boneToModify {boneName=dyn_skirtSide_mstr_r}
    rotation {pitch=15}
   }
   n46 = Get ParachuteAlpha() -> (ParachuteAlpha:double)
   n47 = "Transform (Modify) Bone - Bone: dyn_skirtBack_01_l"(ComponentPose=@n45, Alpha=@n46, Translation="0,0,0", Rotation="15,10,0.000", Scale="1,1,1") {
    rotationMode=BMM_Additive
    rotationSpace=BCS_BoneSpace
    alphaBoolBlend {blendOption=Linear}
    alphaScaleBiasClamp {bInterpResult=true, interpSpeedDecreasing=2, interpSpeedIncreasing=5}
    boneToModify {boneName=dyn_skirtBack_01_l}
    rotation {pitch=15, yaw=10}
   }
   n48 = "Transform (Modify) Bone - Bone: dyn_skirtBack_01_r"(ComponentPose=@n47, Alpha=@n46, Translation="0,0,0", Rotation="15,-10,0.000", Scale="1,1,1") {
    rotationMode=BMM_Additive
    rotationSpace=BCS_BoneSpace
    alphaBoolBlend {blendOption=Linear}
    alphaScaleBiasClamp {bInterpResult=true, interpSpeedDecreasing=2, interpSpeedIncreasing=5}
    boneToModify {boneName=dyn_skirtBack_01_r}
    rotation {pitch=15, yaw=-10}
   }
   n49 = "Transform (Modify) Bone - Bone: dyn_skirtSide_01_l"(ComponentPose=@n48, Alpha=@n46, Translation="0,0,0", Rotation="15,-15,0.000", Scale="1,1,1") {
    rotationMode=BMM_Additive
    rotationSpace=BCS_BoneSpace
    alphaBoolBlend {blendOption=Linear}
    alphaScaleBiasClamp {bInterpResult=true, interpSpeedDecreasing=2, interpSpeedIncreasing=5}
    boneToModify {boneName=dyn_skirtSide_01_l}
    rotation {pitch=15, yaw=-15}
   }
   n50 = "Transform (Modify) Bone - Bone: dyn_skirtSide_01_l"(ComponentPose=@n49, Alpha=@n46, Translation="0,0,0", Rotation="15,15,0.000", Scale="1,1,1") {
    rotationMode=BMM_Additive
    rotationSpace=BCS_BoneSpace
    alphaBoolBlend {blendOption=Linear}
    alphaScaleBiasClamp {bInterpResult=true, interpSpeedDecreasing=2, interpSpeedIncreasing=5}
    boneToModify {boneName=dyn_skirtSide_01_l}
    rotation {pitch=15, yaw=15}
   }
   n51 = AND Boolean(A=Get bIsCrouching(), B=NOT Boolean(A=Get bIsMoving2D()))
   n52 = "Transform (Modify) Bone - Bone: dyn_skirtBack_01_l"!(ComponentPose=@n50, bAlphaBoolEnabled:bool=@n51, Translation="0,0,0", Rotation="15,0,0", Scale="1,1,1") -> (Pose) {
    alphaInputType=Bool
    rotationMode=BMM_Additive
    rotationSpace=BCS_BoneSpace
    scaleMode=BMM_Additive
    scaleSpace=BCS_BoneSpace
    translationMode=BMM_Additive
    translationSpace=BCS_BoneSpace
    alphaBoolBlend {blendInTime=0.1, blendOption=Linear, blendOutTime=0.5}
    boneToModify {boneName=dyn_skirtBack_01_l}
    rotation {pitch=15}
   }
   n53 = "Transform (Modify) Bone - Bone: dyn_skirtBack_01_r"!(ComponentPose=@n52, bAlphaBoolEnabled:bool=@n51, Translation="0,0,0", Rotation="30,0,0", Scale="1,1,1") -> (Pose) {
    alphaInputType=Bool
    rotationMode=BMM_Additive
    rotationSpace=BCS_BoneSpace
    scaleMode=BMM_Additive
    scaleSpace=BCS_BoneSpace
    translationMode=BMM_Additive
    translationSpace=BCS_BoneSpace
    alphaBoolBlend {blendInTime=0.1, blendOption=Linear, blendOutTime=0.5}
    boneToModify {boneName=dyn_skirtBack_01_r}
    rotation {pitch=30}
   }
   n54 = "Transform (Modify) Bone - Bone: dyn_skirtBack_02_r"!(ComponentPose=@n53, bAlphaBoolEnabled:bool=@n51, Translation="0,0,0", Rotation="30,0,0", Scale="1,1,1") -> (Pose) {
    alphaInputType=Bool
    rotationMode=BMM_Additive
    rotationSpace=BCS_BoneSpace
    scaleMode=BMM_Additive
    scaleSpace=BCS_BoneSpace
    translationMode=BMM_Additive
    translationSpace=BCS_BoneSpace
    alphaBoolBlend {blendInTime=0.1, blendOption=Linear, blendOutTime=0.5}
    boneToModify {boneName=dyn_skirtBack_02_r}
    rotation {pitch=30}
   }
   n55 = "Transform (Modify) Bone - Bone: dyn_skirtSide_01_r"!(ComponentPose=@n54, bAlphaBoolEnabled:bool=@n51, Translation="0,0,0", Rotation="30,0,0", Scale="1,1,1") -> (Pose) {
    alphaInputType=Bool
    rotationMode=BMM_Additive
    rotationSpace=BCS_BoneSpace
    scaleMode=BMM_Additive
    scaleSpace=BCS_BoneSpace
    translationMode=BMM_Additive
    translationSpace=BCS_BoneSpace
    alphaBoolBlend {blendInTime=0.1, blendOption=Linear, blendOutTime=0.5}
    boneToModify {boneName=dyn_skirtSide_01_r}
    rotation {pitch=30}
   }
   n56 = AND Boolean(A=Get bIsInFrontEnd(), B=NOT Boolean(A=Get bIsPlayingEmote()))
   n57 = "Constraint - dyn_belt_b_02 (R)"!(ComponentPose=@n55, bAlphaBoolEnabled:bool=@n56, ConstraintWeights_0=1) -> (Pose) {
    alphaInputType=Bool
    alphaBoolBlend {blendInTime=0.1, blendOption=Linear, blendOutTime=0.1}
    boneToModify {boneName=dyn_belt_b_02}
    constraintSetup [{transformType=Rotation, targetBone {boneName=dyn_skirtSide_mstr_l}}]
    constraintWeights [1]
   }
   n58 = "Constraint - dyn_belt_a_02 (R)"!(ComponentPose=@n57, bAlphaBoolEnabled:bool=@n56, ConstraintWeights_0=1) -> (Pose) {
    alphaInputType=Bool
    alphaBoolBlend {blendInTime=0.1, blendOption=Linear, blendOutTime=0.1}
    boneToModify {boneName=dyn_belt_a_02}
    constraintSetup [{transformType=Rotation, targetBone {boneName=dyn_skirtSide_mstr_l}}]
    constraintWeights [1]
   }
   n59 = Control Rig(Source=Blend Poses by bool(BlendPose_0=Use cached pose 'constraints'(), BlendPose_1=Component To Local(ComponentPose=@n58), bActiveValue=Get bRigidBodyEnabled()) {blendType=Linear}) {
    alphaBoolBlend {blendOption=Linear}
    controlRigAssetReference {blueprintRigClass=&/F_MED_Thrive_Spirit_CtrlRig.F_MED_Thrive_Spirit_CtrlRig_C}
   }
   Save cached pose 'prerigid'(Pose=@n59) {cacheName=prerigid}
   n60 = RigidBody(ComponentPose=Local To Component(LocalPose=Use cached pose 'prerigid'()), OverrideWorldGravity=Get GravityOverride()) {
    bClampLinearTranslationLimitToRefPose=true
    bOverrideWorldGravity=true
    overridePhysicsAsset=&F_MED_Thrive_Spirit_Skirt_Physics
    simulationSpace=BaseBoneSpace
    alphaBoolBlend {blendOption=Linear}
    baseBoneRef {boneName=pelvis}
    componentAppliedLinearAccClamp=<1500 1500 1500>
    componentLinearAccScale=<0.25 0.25 0.4>
    componentLinearVelScale=<0.2 0.2 0.2>
   }
   n61 = RigidBody(ComponentPose=@n60, OverrideWorldGravity=Get GravityOverride_ropes()) {
    bClampLinearTranslationLimitToRefPose=true
    bOverrideWorldGravity=true
    overridePhysicsAsset=&/F_MED_Thrive_Spirit_Rope_Physics
    simulationSpace=BaseBoneSpace
    alphaBoolBlend {blendOption=Linear}
    baseBoneRef {boneName=pelvis}
    componentAppliedLinearAccClamp=<3000 3000 3000>
    componentLinearAccScale=<0.3 0.3 0.3>
    componentLinearVelScale=<0.3 0.3 0.3>
    overrideWorldGravity {z=-980}
    simSpaceSettings {
     maxAngularAcceleration=1000
     maxAngularVelocity=3
     maxLinearAcceleration=700
     maxLinearVelocity=500
     velocityScaleZ=0
     worldAlpha=1
     externalLinearDragV=<1 1>
    }
   }
   n62 = RigidBody(ComponentPose=Local To Component(LocalPose=Use cached pose 'prerigid'()), OverrideWorldGravity=Get GravityOverride()) {
    bClampLinearTranslationLimitToRefPose=true
    bOverrideWorldGravity=true
    overridePhysicsAsset=&F_MED_Thrive_Spirit_Skirt_Physics
    simulationSpace=BaseBoneSpace
    alphaBoolBlend {blendOption=Linear}
    baseBoneRef {boneName=dyn_root_simspace}
    componentAppliedLinearAccClamp=<1500 1500 1500>
    componentLinearAccScale=<0.25 0.25 0.4>
    componentLinearVelScale=<0.2 0.2 0.2>
   }
   n63 = RigidBody(ComponentPose=@n62, OverrideWorldGravity=Get GravityOverride_ropes()) {
    bClampLinearTranslationLimitToRefPose=true
    bOverrideWorldGravity=true
    overridePhysicsAsset=&F_MED_Thrive_Spirit_Rope_Physics
    simulationSpace=BaseBoneSpace
    alphaBoolBlend {blendOption=Linear}
    baseBoneRef {boneName=dyn_root_simspace}
    componentAppliedLinearAccClamp=<3000 3000 3000>
    componentLinearAccScale=<0.3 0.3 0.3>
    componentLinearVelScale=<0.3 0.3 0.3>
    overrideWorldGravity {z=-980}
    simSpaceSettings {
     maxAngularAcceleration=1000
     maxAngularVelocity=3
     maxLinearAcceleration=700
     maxLinearVelocity=500
     velocityScaleZ=0
     worldAlpha=1
     externalLinearDragV=<1 1>
    }
   }
   n64 = Blend Poses by bool(BlendPose_0=Component To Local(ComponentPose=@n61), BlendPose_1=Component To Local(ComponentPose=@n63), bActiveValue=Get bIsSkyDiving()) {blendType=Linear}
   n65 = RigidBody(ComponentPose=Local To Component(LocalPose=Use cached pose 'prerigid'()), OverrideWorldGravity=Get GravityOverride()) {
    bClampLinearTranslationLimitToRefPose=true
    bOverrideWorldGravity=true
    overridePhysicsAsset=&F_MED_Thrive_Spirit_Skirt_Physics
    simulationSpace=BaseBoneSpace
    alphaBoolBlend {blendOption=Linear}
    baseBoneRef {boneName=dyn_root_simspace}
    componentAppliedLinearAccClamp=<1500 1500 1500>
    componentLinearAccScale=<0.25 0.25 0.4>
    componentLinearVelScale=<0.2 0.2 0.2>
   }
   n66 = RigidBody(ComponentPose=@n65, OverrideWorldGravity="0,0,-980") {
    bClampLinearTranslationLimitToRefPose=true
    bOverrideWorldGravity=true
    overridePhysicsAsset=&F_MED_Thrive_Spirit_Rope_Emote_Physics
    simulationSpace=BaseBoneSpace
    alphaBoolBlend {blendOption=Linear}
    baseBoneRef {boneName=dyn_root_simspace}
    componentAppliedLinearAccClamp=<3000 3000 3000>
    componentLinearAccScale=<0.3 0.3 0.3>
    componentLinearVelScale=<0.3 0.3 0.3>
    overrideWorldGravity {z=-980}
    simSpaceSettings {
     maxAngularAcceleration=1000
     maxAngularVelocity=3
     maxLinearAcceleration=700
     maxLinearVelocity=500
     velocityScaleZ=0
     worldAlpha=1
     externalLinearDragV=<1 1>
    }
   }
   n67 = RigidBody(ComponentPose=Local To Component(LocalPose=Use cached pose 'prerigid'()), OverrideWorldGravity=Get GravityOverride()) {
    bClampLinearTranslationLimitToRefPose=true
    bOverrideWorldGravity=true
    overridePhysicsAsset=&/F_MED_Thrive_Spirit_Skirt_emote_Physics
    simulationSpace=BaseBoneSpace
    alphaBoolBlend {blendOption=Linear}
    baseBoneRef {boneName=pelvis}
    componentAppliedLinearAccClamp=<1500 1500 1500>
    componentLinearAccScale=<0.25 0.25 0.4>
    componentLinearVelScale=<0.2 0.2 0.2>
    simSpaceSettings {maxAngularAcceleration=1000, maxAngularVelocity=3, maxLinearAcceleration=700, maxLinearVelocity=500, velocityScaleZ=0, worldAlpha=1}
   }
   n68 = RigidBody(ComponentPose=@n67, OverrideWorldGravity=Get GravityOverride_ropes()) {
    bClampLinearTranslationLimitToRefPose=true
    bOverrideWorldGravity=true
    overridePhysicsAsset=&F_MED_Thrive_Spirit_Rope_Physics
    simulationSpace=BaseBoneSpace
    alphaBoolBlend {blendOption=Linear}
    baseBoneRef {boneName=dyn_root_simspace}
    componentAppliedLinearAccClamp=<3000 3000 3000>
    componentLinearAccScale=<0.3 0.3 0.3>
    componentLinearVelScale=<0.3 0.3 0.3>
    overrideWorldGravity {z=-980}
    simSpaceSettings {
     maxAngularAcceleration=1000
     maxAngularVelocity=3
     maxLinearAcceleration=700
     maxLinearVelocity=500
     velocityScaleZ=0
     worldAlpha=1
     externalLinearDragV=<1 1>
    }
   }
   n69 = Blend Poses by bool(
    BlendPose_0=Component To Local(ComponentPose=@n66)
    BlendPose_1=Component To Local(ComponentPose=@n68)
    bActiveValue=Get thrive emote() -> (thrive emote:bool)
   ) {blendType=Linear}
   n70 = Get bIsPlayingEmote()
   n71 = Anim Dynamics - Bone: dyn_arm_band_r_MSTR(
    ComponentPose=Local To Component(LocalPose=Blend Poses by bool(BlendPose_0=@n69, BlendPose_1=@n64, bActiveValue=@n70) {blendType=Linear})
    Alpha=1
    ExternalForce="0,0,0"
   ) {
    angularDampingOverride=1
    angularSpringConstant=20000
    bAngularSpring=true
    bOverrideAngularDamping=true
    bUseGravityOverride=true
    gravityScale=0.35
    simulationSpace=RootRelative
    alphaBoolBlend {blendOption=Linear}
    boundBone {boneName=dyn_arm_band_r_MSTR}
    componentLinearAccScale=<1 1 1>
    componentLinearVelScale=<1 1 1>
    gravityOverride {z=-980}
    physicsBodyDefinitions [{
      sphereCollisionRadius=0
      boundBone {boneName=dyn_arm_band_r_MSTR}
      boxExtents=<30 40 40>
      constraintSetup {angularLimitsMax=<15 15 15>, angularLimitsMin=<-15 -15 -15>, angularTarget=<1 0 0>}
      localJointOffset=<-2 0 0>
     }]
    relativeSpaceBone {boneName=root}
   }
   n72 = Anim Dynamics - Bone: dyn_arm_band_l_MSTR(ComponentPose=@n71, Alpha=1, ExternalForce="0,0,0") {
    angularDampingOverride=1
    angularSpringConstant=20000
    bAngularSpring=true
    bOverrideAngularDamping=true
    bUseGravityOverride=true
    gravityScale=0.35
    simulationSpace=RootRelative
    alphaBoolBlend {blendOption=Linear}
    boundBone {boneName=dyn_arm_band_l_MSTR}
    componentLinearAccScale=<1 1 1>
    componentLinearVelScale=<1 1 1>
    gravityOverride {z=-980}
    physicsBodyDefinitions [{
      sphereCollisionRadius=0
      boundBone {boneName=dyn_arm_band_l_MSTR}
      boxExtents=<30 40 40>
      constraintSetup {angularLimitsMax=<15 15 15>, angularLimitsMin=<-15 -15 -15>, angularTarget=<1 0 0>}
      localJointOffset=<2 0 0>
     }]
    relativeSpaceBone {boneName=root}
   }
   n73 = Anim Dynamics - Bone: dyn_r_cuff_02(ComponentPose=@n72, Alpha=1, ExternalForce="0,0,0") {
    angularDampingOverride=1
    angularSpringConstant=20000
    bAngularSpring=true
    bOverrideAngularDamping=true
    bUseGravityOverride=true
    gravityScale=0.35
    simulationSpace=RootRelative
    alphaBoolBlend {blendOption=Linear}
    boundBone {boneName=dyn_r_cuff_02}
    componentLinearAccScale=<1 1 1>
    componentLinearVelScale=<1 1 1>
    gravityOverride {z=-980}
    physicsBodyDefinitions [{
      sphereCollisionRadius=0
      boundBone {boneName=dyn_r_cuff_02}
      boxExtents=<30 40 40>
      constraintSetup {angularLimitsMax=<15 15 15>, angularLimitsMin=<-15 -15 -15>, angularTarget=<1 0 0>}
      localJointOffset=<-2 0 0>
     }]
    relativeSpaceBone {boneName=root}
   }
   n74 = Anim Dynamics - Bone: dyn_l_cuff_02(ComponentPose=@n73, Alpha=1, ExternalForce="0,0,0") {
    angularDampingOverride=1
    angularSpringConstant=20000
    bAngularSpring=true
    bOverrideAngularDamping=true
    bUseGravityOverride=true
    gravityScale=0.35
    simulationSpace=RootRelative
    alphaBoolBlend {blendOption=Linear}
    boundBone {boneName=dyn_l_cuff_02}
    componentLinearAccScale=<1 1 1>
    componentLinearVelScale=<1 1 1>
    gravityOverride {z=-980}
    physicsBodyDefinitions [{
      sphereCollisionRadius=0
      boundBone {boneName=dyn_l_cuff_02}
      boxExtents=<30 40 40>
      constraintSetup {angularLimitsMax=<15 15 15>, angularLimitsMin=<-15 -15 -15>, angularTarget=<1 0 0>}
      localJointOffset=<2 0 0>
     }]
    relativeSpaceBone {boneName=root}
   }
   Save cached pose 'post rigid'(Pose=Component To Local(ComponentPose=@n74)) {cacheName=post rigid}
   n75 = Get skirt fixer() -> (skirt fixer:bool)
   n76 = "Transform (Modify) Bone - Bone: dyn_skirtFront_mstr"!(
    ComponentPose=Local To Component(LocalPose=Copy Pose From Mesh() {bUseAttachedParent=true})
    bAlphaBoolEnabled:bool=True
    Translation="-0.000,8,2"
    Rotation="0,0,45"
    Scale="1,1,1"
   ) -> (Pose) {
    alphaInputType=Bool
    bAlphaBoolEnabled=false
    rotationMode=BMM_Additive
    rotationSpace=BCS_BoneSpace
    scaleMode=BMM_Additive
    scaleSpace=BCS_BoneSpace
    translationMode=BMM_Additive
    translationSpace=BCS_BoneSpace
    alphaBoolBlend {blendInTime=0.1, blendOption=Linear, blendOutTime=0.5}
    boneToModify {boneName=dyn_skirtFront_mstr}
    rotation {roll=45}
    translation {y=8, z=2}
   }
   n77 = "Transform (Modify) Bone - Bone: dyn_skirtFront_02_l"(ComponentPose=@n76, Alpha=Get thigh_rot_l(), Translation="0,0,3", Rotation="22,-20,-15", Scale="1,1,1") {
    rotationMode=BMM_Additive
    rotationSpace=BCS_BoneSpace
    scaleMode=BMM_Additive
    scaleSpace=BCS_BoneSpace
    translationMode=BMM_Additive
    translationSpace=BCS_BoneSpace
    alphaBoolBlend {blendOption=Linear}
    boneToModify {boneName=dyn_skirtFront_02_l}
    rotation=<rot 22 -20 -15>
    translation {z=3}
   }
   n78 = "Constraint - dyn_skirtFront_mstr_r (PP)"!(ComponentPose=@n77, Alpha=1, ConstraintWeights_0=0.8, ConstraintWeights_1:float=0.2) -> (Pose) {
    alphaBoolBlend {blendOption=Linear}
    boneToModify {boneName=dyn_skirtFront_mstr_r}
    constraintSetup [{targetBone {boneName=thigh_twist_01_r}}, {targetBone {boneName=pelvis}}]
    constraintWeights [1,1]
   }
   n79 = "Transform (Modify) Bone - Bone: dyn_shirt_fixer"(ComponentPose=@n78, Alpha=Get thigh_rot_r(), Translation="-5,0,0", Rotation="0,0,0", Scale="1,1,1") {
    rotationMode=BMM_Additive
    rotationSpace=BCS_BoneSpace
    scaleMode=BMM_Additive
    scaleSpace=BCS_BoneSpace
    translationMode=BMM_Additive
    translationSpace=BCS_BoneSpace
    alphaBoolBlend {blendOption=Linear}
    boneToModify {boneName=dyn_shirt_fixer}
    translation {x=-5}
   }
   n80 = "Transform (Modify) Bone - Bone: dyn_necklace_helper"(ComponentPose=@n79, Alpha=1, Translation="1,0,0", Rotation="0,0,0", Scale="1,1,1") {
    rotationMode=BMM_Additive
    rotationSpace=BCS_BoneSpace
    scaleMode=BMM_Additive
    scaleSpace=BCS_BoneSpace
    translationMode=BMM_Additive
    translationSpace=BCS_BoneSpace
    alphaBoolBlend {blendOption=Linear}
    boneToModify {boneName=dyn_necklace_helper}
    translation {x=1}
   }
   n81 = "Transform (Modify) Bone - Bone: dyn_shirt_fixer_02"(ComponentPose=@n80, Alpha=Get thigh_rot_l(), Translation="-3,1,0", Rotation="-20.000,0.000,0.000", Scale="1,1,1") {
    rotationMode=BMM_Additive
    rotationSpace=BCS_BoneSpace
    scaleMode=BMM_Additive
    scaleSpace=BCS_BoneSpace
    translationMode=BMM_Additive
    translationSpace=BCS_BoneSpace
    alphaBoolBlend {blendOption=Linear}
    boneToModify {boneName=dyn_shirt_fixer_02}
    rotation {pitch=-20}
    translation=<-3 1>
   }
   n82 = "Transform (Modify) Bone - Bone: dyn_skirtFront_mstr_r"!(ComponentPose=@n81, bAlphaBoolEnabled:bool=@n75, Translation="0,0,0", Rotation="2,2,0", Scale="1,1,1") -> (Pose) {
    alphaInputType=Bool
    rotationMode=BMM_Additive
    rotationSpace=BCS_BoneSpace
    scaleMode=BMM_Additive
    scaleSpace=BCS_BoneSpace
    translationMode=BMM_Additive
    translationSpace=BCS_BoneSpace
    alphaBoolBlend {blendInTime=0.1, blendOption=Linear, blendOutTime=0.1}
    alphaScaleBiasClamp {bMapRange=true}
    boneToModify {boneName=dyn_skirtFront_mstr_r}
    rotation {pitch=2, yaw=2}
   }
   n83 = "Transform (Modify) Bone - Bone: dyn_skirtFront_mstr_r"(ComponentPose=@n82, bAlphaBoolEnabled=Get bIsMoving2D(), Translation="0,0,0", Rotation="15,7,0", Scale="1,1,1") {
    alphaInputType=Bool
    rotationMode=BMM_Additive
    rotationSpace=BCS_BoneSpace
    scaleMode=BMM_Additive
    scaleSpace=BCS_BoneSpace
    translationMode=BMM_Additive
    translationSpace=BCS_BoneSpace
    alphaBoolBlend {blendInTime=0.1, blendOption=Linear, blendOutTime=0.1}
    alphaScaleBiasClamp {bMapRange=true}
    boneToModify {boneName=dyn_skirtFront_mstr_r}
    rotation {pitch=15, yaw=7}
   }
   n84 = "Transform (Modify) Bone - Bone: dyn_skirtFront_mstr_r"(ComponentPose=@n83, bAlphaBoolEnabled=True, Translation="0,0,0", Rotation="10,8,0", Scale="1,1,1") {
    alphaInputType=Bool
    bAlphaBoolEnabled=false
    rotationMode=BMM_Additive
    rotationSpace=BCS_BoneSpace
    scaleMode=BMM_Additive
    scaleSpace=BCS_BoneSpace
    translationMode=BMM_Additive
    translationSpace=BCS_BoneSpace
    alphaBoolBlend {blendInTime=0.1, blendOption=Linear, blendOutTime=0.1}
    alphaScaleBiasClamp {bMapRange=true}
    boneToModify {boneName=dyn_skirtFront_mstr_r}
    rotation {pitch=10, yaw=8}
   }
   n85 = "Transform (Modify) Bone - Bone: dyn_skirtSide_mstr_l"(ComponentPose=@n84, bAlphaBoolEnabled=@n75, Translation="0,0,0", Rotation="5.000,0.000,0.000", Scale="1,1,1") {
    alphaInputType=Bool
    rotationMode=BMM_Additive
    rotationSpace=BCS_BoneSpace
    scaleMode=BMM_Additive
    scaleSpace=BCS_BoneSpace
    translationMode=BMM_Additive
    translationSpace=BCS_BoneSpace
    alphaBoolBlend {blendInTime=0.1, blendOption=Linear, blendOutTime=0.1}
    alphaScaleBiasClamp {bMapRange=true}
    boneToModify {boneName=dyn_skirtSide_mstr_l}
    rotation {pitch=5}
   }
   n86 = "Transform (Modify) Bone - Bone: dyn_belt_B_MSTR"!(ComponentPose=@n85, bAlphaBoolEnabled:bool=Get ropecrouch() -> (ropecrouch:bool), Translation="0,0,0", Rotation="15,5,-25", Scale="1,1,1") -> (Pose) {
    alphaInputType=Bool
    rotationMode=BMM_Additive
    rotationSpace=BCS_BoneSpace
    scaleMode=BMM_Additive
    scaleSpace=BCS_BoneSpace
    translationMode=BMM_Additive
    translationSpace=BCS_BoneSpace
    alphaBoolBlend {blendInTime=0.1, blendOption=Linear, blendOutTime=0.1}
    boneToModify {boneName=dyn_belt_B_MSTR}
    rotation=<rot 15 5 -25>
   }
   n87 = "Transform (Modify) Bone - Bone: dyn_skirtFront_mstr_r"!(ComponentPose=@n86, Alpha:float=Get thigh_rot_r(), Translation="0,0,0", Rotation="20,15.001,0.000", Scale="1,1,1") -> (Pose) {
    rotationMode=BMM_Additive
    rotationSpace=BCS_BoneSpace
    scaleMode=BMM_Additive
    scaleSpace=BCS_BoneSpace
    translationMode=BMM_Additive
    translationSpace=BCS_BoneSpace
    alphaBoolBlend {blendOption=Linear}
    boneToModify {boneName=dyn_skirtFront_mstr_r}
    rotation {pitch=20, yaw=15.001}
   }
   n88 = "Transform (Modify) Bone - Bone: dyn_skirtFront_mstr_r"(ComponentPose=@n87, Alpha=Get thigh_rot_otr_r() -> (thigh_rot_otr_r:double), Translation="0,0,0", Rotation="20,15.001,0.000", Scale="1,1,1") {
    rotationMode=BMM_Additive
    rotationSpace=BCS_BoneSpace
    scaleMode=BMM_Additive
    scaleSpace=BCS_BoneSpace
    translationMode=BMM_Additive
    translationSpace=BCS_BoneSpace
    alphaBoolBlend {blendOption=Linear}
    boneToModify {boneName=dyn_skirtFront_mstr_r}
    rotation {pitch=20, yaw=15.001}
   }
   n89 = "Transform (Modify) Bone - Bone: dyn_skirtFront_mstr_r"!(ComponentPose=@n88, bAlphaBoolEnabled:bool=True, Translation="0,0,0", Rotation="10,0,0.000", Scale="1,1,1") -> (Pose) {
    alphaInputType=Bool
    bAlphaBoolEnabled=false
    rotationMode=BMM_Additive
    rotationSpace=BCS_BoneSpace
    scaleMode=BMM_Additive
    scaleSpace=BCS_BoneSpace
    translationMode=BMM_Additive
    translationSpace=BCS_BoneSpace
    alphaBoolBlend {blendInTime=0.1, blendOption=Linear, blendOutTime=0.1}
    boneToModify {boneName=dyn_skirtFront_mstr_r}
    rotation {pitch=10}
   }
   n90 = "Transform (Modify) Bone - Bone: dyn_skirtFront_mstr"(ComponentPose=@n89, bAlphaBoolEnabled=Get dance moves() -> (dance moves:bool), Translation="0,0,0", Rotation="5,0.000,0.000", Scale="1,1,1") {
    alphaInputType=Bool
    rotationMode=BMM_Additive
    rotationSpace=BCS_BoneSpace
    scaleMode=BMM_Additive
    scaleSpace=BCS_BoneSpace
    translationMode=BMM_Additive
    translationSpace=BCS_BoneSpace
    alphaBoolBlend {blendInTime=0.1, blendOption=Linear, blendOutTime=0.1}
    boneToModify {boneName=dyn_skirtFront_mstr}
    rotation {pitch=5}
   }
   n91 = "Transform (Modify) Bone - Bone: dyn_r_cuff_01"(ComponentPose=@n90, Alpha=Get arm r Rot() -> (arm r Rot:double), Translation="0,0,0", Rotation="0,-80,0", Scale="1,1,1") {
    rotationMode=BMM_Additive
    rotationSpace=BCS_BoneSpace
    scaleMode=BMM_Additive
    scaleSpace=BCS_BoneSpace
    translationMode=BMM_Additive
    translationSpace=BCS_BoneSpace
    alphaBoolBlend {blendOption=Linear}
    boneToModify {boneName=dyn_r_cuff_01}
    rotation {yaw=-80}
   }
   n92 = "Transform (Modify) Bone - Bone: dyn_l_cuff_01"(ComponentPose=@n91, Alpha=Get arm l Rot() -> (arm l Rot:double), Translation="0,0,0", Rotation="0,60,0", Scale="1,1,1") {
    rotationMode=BMM_Additive
    rotationSpace=BCS_BoneSpace
    scaleMode=BMM_Additive
    scaleSpace=BCS_BoneSpace
    translationMode=BMM_Additive
    translationSpace=BCS_BoneSpace
    alphaBoolBlend {blendOption=Linear}
    boneToModify {boneName=dyn_l_cuff_01}
    rotation {yaw=60}
   }
   n93 = Get bIsJumping()
   n94 = "Transform (Modify) Bone - Bone: dyn_skirtFront_01"!(ComponentPose=@n92, bAlphaBoolEnabled:bool=@n93, Translation="0,0,0", Rotation="5,15,0", Scale="1,1,1") -> (Pose) {
    alphaInputType=Bool
    rotationMode=BMM_Additive
    rotationSpace=BCS_BoneSpace
    scaleMode=BMM_Additive
    scaleSpace=BCS_BoneSpace
    translationMode=BMM_Additive
    translationSpace=BCS_BoneSpace
    alphaBoolBlend {blendInTime=0.1, blendOption=Linear, blendOutTime=0.8}
    boneToModify {boneName=dyn_skirtFront_01}
    rotation {pitch=5, yaw=15}
   }
   n95 = "Transform (Modify) Bone - Bone: dyn_skirtFront_mstr_r"(ComponentPose=@n94, bAlphaBoolEnabled=@n93, Translation="0,0,0", Rotation="10,8,0", Scale="1,1,1") {
    alphaInputType=Bool
    rotationMode=BMM_Additive
    rotationSpace=BCS_BoneSpace
    scaleMode=BMM_Additive
    scaleSpace=BCS_BoneSpace
    translationMode=BMM_Additive
    translationSpace=BCS_BoneSpace
    alphaBoolBlend {blendInTime=0.1, blendOption=Linear, blendOutTime=0.8}
    alphaScaleBiasClamp {bMapRange=true}
    boneToModify {boneName=dyn_skirtFront_mstr_r}
    rotation {pitch=10, yaw=8}
   }
   n96 = "Transform (Modify) Bone - Bone: dyn_floor_COL"!(
    ComponentPose=@n95
    bAlphaBoolEnabled:bool=OR Boolean!(A=Get bIsFalling(), B=false, C:bool=Get bIsDBNOCarried()) -> (ReturnValue)
    Translation="0,0,-300"
    Rotation="0,0,0"
    Scale="1,1,1"
   ) -> (Pose) {
    alphaInputType=Bool
    bAlphaBoolEnabled=false
    translationMode=BMM_Additive
    translationSpace=BCS_BoneSpace
    alphaBoolBlend {blendInTime=0.1, blendOption=Linear, blendOutTime=0.1}
    boneToModify {boneName=dyn_floor_COL}
    translation {z=-300}
   }
   n97 = "Transform (Modify) Bone - Bone: dyn_floor_COL"(ComponentPose=@n96, bAlphaBoolEnabled=Get bIsSliding(), Translation="0.000,-15,-20", Rotation="0,0,0", Scale="1,1,1") {
    alphaInputType=Bool
    bAlphaBoolEnabled=false
    translationMode=BMM_Additive
    translationSpace=BCS_BoneSpace
    alphaBoolBlend {blendInTime=0.1, blendOption=Linear, blendOutTime=0.1}
    boneToModify {boneName=dyn_floor_COL}
    translation {y=-15, z=-20}
   }
   n98 = "Transform (Modify) Bone - Bone: dyn_skirtSide_mstr_l"(ComponentPose=@n97, bAlphaBoolEnabled=True, Translation="0,0,0", Rotation="85.000,0.000,0.000", Scale="1,1,1") {
    alphaInputType=Bool
    bAlphaBoolEnabled=false
    rotationMode=BMM_Additive
    rotationSpace=BCS_BoneSpace
    scaleMode=BMM_Additive
    scaleSpace=BCS_BoneSpace
    translationMode=BMM_Additive
    translationSpace=BCS_BoneSpace
    alphaBoolBlend {blendInTime=0.3, blendOption=Linear, blendOutTime=0.1}
    boneToModify {boneName=dyn_skirtSide_mstr_l}
    rotation {pitch=85}
   }
   n99 = "Transform (Modify) Bone - Bone: dyn_skirtSide_mstr_r"!(ComponentPose=@n98, bAlphaBoolEnabled:bool=True, Translation="0,0,0", Rotation="85.000,0.000,0.000", Scale="1,1,1") -> (Pose) {
    alphaInputType=Bool
    bAlphaBoolEnabled=false
    rotationMode=BMM_Additive
    rotationSpace=BCS_BoneSpace
    scaleMode=BMM_Additive
    scaleSpace=BCS_BoneSpace
    translationMode=BMM_Additive
    translationSpace=BCS_BoneSpace
    alphaBoolBlend {blendInTime=0.3, blendOption=Linear, blendOutTime=0.1}
    boneToModify {boneName=dyn_skirtSide_mstr_r}
    rotation {pitch=85}
   }
   n100 = "Transform (Modify) Bone - Bone: dyn_skirtBack_mstr_l"!(ComponentPose=@n99, bAlphaBoolEnabled:bool=True, Translation="0,0,0", Rotation="120,0.000,0.000", Scale="1,1,1") -> (Pose) {
    alphaInputType=Bool
    bAlphaBoolEnabled=false
    rotationMode=BMM_Additive
    rotationSpace=BCS_BoneSpace
    scaleMode=BMM_Additive
    scaleSpace=BCS_BoneSpace
    translationMode=BMM_Additive
    translationSpace=BCS_BoneSpace
    alphaBoolBlend {blendInTime=0.3, blendOption=Linear, blendOutTime=0.1}
    boneToModify {boneName=dyn_skirtBack_mstr_l}
    rotation {pitch=120}
   }
   n101 = "Transform (Modify) Bone - Bone: dyn_skirtBack_mstr_r"!(ComponentPose=@n100, bAlphaBoolEnabled:bool=True, Translation="0,0,0", Rotation="120,0.000,0.000", Scale="1,1,1") -> (Pose) {
    alphaInputType=Bool
    bAlphaBoolEnabled=false
    rotationMode=BMM_Additive
    rotationSpace=BCS_BoneSpace
    scaleMode=BMM_Additive
    scaleSpace=BCS_BoneSpace
    translationMode=BMM_Additive
    translationSpace=BCS_BoneSpace
    alphaBoolBlend {blendInTime=0.3, blendOption=Linear, blendOutTime=0.1}
    boneToModify {boneName=dyn_skirtBack_mstr_r}
    rotation {pitch=120}
   }
   n102 = Get skydiveAlpha() -> (skydiveAlpha:double)
   n103 = "Transform (Modify) Bone - Bone: dyn_belt_B_MSTR"!(ComponentPose=@n101, Alpha:float=@n102, Translation="0,0,0", Rotation="35.000,0.000,-0.000", Scale="1,1,1") -> (Pose) {
    rotationMode=BMM_Additive
    rotationSpace=BCS_BoneSpace
    alphaBoolBlend {blendOption=Linear}
    alphaScaleBiasClamp {bInterpResult=true, interpSpeedDecreasing=2, interpSpeedIncreasing=5}
    boneToModify {boneName=dyn_belt_B_MSTR}
    rotation=<rot 34.9999 0 -0.0001>
   }
   n104 = "Transform (Modify) Bone - Bone: dyn_belt_a_MSTR"(ComponentPose=@n103, Alpha=@n102, Translation="0,0,0", Rotation="10.546,40,10.729", Scale="1,1,1") {
    rotationMode=BMM_Additive
    rotationSpace=BCS_BoneSpace
    alphaBoolBlend {blendOption=Linear}
    alphaScaleBiasClamp {bInterpResult=true, interpSpeedDecreasing=2, interpSpeedIncreasing=5}
    boneToModify {boneName=dyn_belt_a_MSTR}
    rotation=<rot 10.546 40 10.729>
   }
   Save cached pose 'constraints'(Pose=Component To Local(ComponentPose=@n104)) {cacheName=constraints}
   Get Head Rot_2()
   Get thigh_rot_otr_r()
   Get GravityOverride_ropes()
   Get thigh_rot_r()
  ]}
 EventGraph {nodes [
   n1 = Random Vector Function!(
    execute
    CharacterPartAnimInstance:Custom Character Part Anim Instance Object Reference=Self-Reference()
    SkelMeshComponent:Mesh Component Object Reference
    JointName:Name=None
    GameTimeInSeconds:double=GetGameTimeInSeconds()
    TimeOffset:double=0
    PerlinScale:double=0
    Frequency:Vector="0.1,0.2,1"
    Amplitude:Vector="0.3,0.7,10"
    Offset:Vector="0,0,0"
   ) -> (then, VectorOut:Vector)
   > Set crown movement(crown movement:Vector=@n1) -> (Output_Get:Vector)
   Event BlueprintUpdateAnimation()
   > Get!(execute:Exec) -> (then:Exec, else:Exec, Skel_Mesh:Skeletal Mesh Component Object Reference)
   > n2 = Sequence!(execute) -> (then_0, then_1, then_2:Exec, then_3:Exec, then_4:Exec, then_5:Exec, then_6:Exec, then_7:Exec, then_8:Exec, then_9:Exec, then_10:Exec)
   > Set skirt fixer(skirt fixer:bool=true) -> (Output_Get:bool)
   n3 = Get Skel_Mesh()
   > Set thigh_rot_l(thigh_rot_l=GetAlphaFromReferenceTransformDelta(SkeletalMeshComponent=@n3, BoneName=thigh_l, BaseName=pelvis, InputRange=(X=-20,Y=-100), TwistAxis="1,0,0"))
   > Set thigh_rot_r(thigh_rot_r=GetAlphaFromReferenceTransformDelta(SkeletalMeshComponent=@n3, BoneName=thigh_r, BaseName=pelvis, InputRange=(X=-20,Y=-100), TwistAxis="1,0,0"))
   > Set arm l Rot(arm l Rot:double=GetAlphaFromReferenceTransformDelta(SkeletalMeshComponent=@n3, BoneName=lowerarm_l, BaseName=upperarm_l, InputRange=(X=50,Y=90), TwistAxis="1,0,0")) -> (Output_Get:double)
   > Set arm r Rot(arm r Rot:double=GetAlphaFromReferenceTransformDelta(SkeletalMeshComponent=@n3, BoneName=lowerarm_r, BaseName=upperarm_r, InputRange=(X=50,Y=90), TwistAxis="1,0,0")) -> (Output_Get:double)
   > Set Head Rot_2(Head Rot_2:double=GetAlphaFromReferenceTransformDelta(
     SkeletalMeshComponent=Get Skel_Mesh()
     BoneName=head
     BaseName=spine_05
     DeltaComponentType=EulerX
     InputRange=(X=20,Y=40)
     TwistAxis="1,0,0"
    )) -> (Output_Get:double)
   n4 = Get Skel_Mesh()
   > Set thigh_rot_l_back(thigh_rot_l_back:double=GetAlphaFromReferenceTransformDelta(SkeletalMeshComponent=@n4, BoneName=thigh_l, BaseName=pelvis, InputRange=(X=0,Y=25), TwistAxis="1,0,0")) -> (Output_Get:double)
   > Set thigh_rot_r_back(thigh_rot_r_back:double=GetAlphaFromReferenceTransformDelta(SkeletalMeshComponent=@n4, BoneName=thigh_r, BaseName=pelvis, InputRange=(X=0,Y=25), TwistAxis="1,0,0")) -> (Output_Get:double)
   > n5 = Cast To FortPlayerAnimInstance(Object=GetAnimInstance(self=GetOwnerSkeletalMeshComponent()))
   > Set pistol(pistol:bool=false) -> (Output_Get:bool)
   > Set rifle(rifle:bool=false) -> (Output_Get:bool)
   > n6 = Switch on EFortWeaponCoreAnimation!(execute, Selection:EFortWeaponCoreAnimation Enum=Get WeaponCoreAnim(self=@n5)) -> (
    Melee:Exec
    Pistol:Exec
    Shotgun:Exec
    PaperBlueprint:Exec
    Rifle:Exec
    MeleeOneHand:Exec
    MachinePistol:Exec
    RocketLauncher:Exec
    GrenadeLauncher:Exec
    GoingCommando:Exec
    AssaultRifle:Exec
    TacticalShotgun:Exec
    SniperRifle:Exec
    TrapPlacement:Exec
    ShoulderLauncher:Exec
    AbilityDecoTool:Exec
    Crossbow:Exec
    C4:Exec
    RemoteControl:Exec
    DualWield:Exec
    AR_BullPup:Exec
    AR_ForwardGrip:Exec
    MedPackPaddles:Exec
    SMG_P90:Exec
    AR_DrumGun:Exec
    Consumable_Small:Exec
    Consumable_Large:Exec
    Balloon:Exec
    MountedTurret:Exec
    CreativeTool:Exec
    ExplosiveBow:Exec
    AshtonIndigo:Exec
    AshtonChicago:Exec
    MeleeDualWield:Exec
    Unarmed:Exec
   )
   n7 = Branch(
    execute=[@n6.AssaultRifle, @n6.Crossbow, @n6.GrenadeLauncher, @n6.Rifle, @n6.RocketLauncher, @n6.Shotgun, @n6.SniperRifle, @n6.TacticalShotgun]
    Condition=Get bIsTargeting()
   )
   > Set rifle(rifle=true)
   Set rifle(execute=@n7.else, rifle=false)
   Branch(execute=[@n6.MachinePistol, @n6.Pistol])
   > Set pistol(pistol=true)
   > Set Necklace 01(Necklace 01:double=GetAlphaFromReferenceTransformDelta(SkeletalMeshComponent=Get Skel_Mesh(), BoneName=upperarm_r, BaseName=spine_05, InputRange=(X=90,Y=125), TwistAxis="1,0,0")) -> (Output_Get:double)
   n8 = GetRealTimeSeconds()
   n9 = "Sin (Radians)"(A=@n8)
   n10 = Get bIsTargeting()
   n11 = SelectFloat(A=0.5, B=3, bPickA=@n10)
   Set sineRotation3(execute=@n2.then_1, sineRotation3:Rotator=MakeRotator(Pitch=float * float(A=@n9, B=@n11))) -> (Output_Get:Rotator)
   n12 = "Sin (Radians)"(A=float + float(A=@n8, B=-1))
   n13 = SelectFloat(A=1, B=7, bPickA=@n10)
   > Set sineRotation4(sineRotation4:Rotator=MakeRotator(Pitch=float * float(A=@n12, B=@n13))) -> (Output_Get:Rotator)
   n14 = "Sin (Radians)"(A=float + float(A=@n8, B=-3))
   > Set sineRotation5(sineRotation5:Rotator=MakeRotator(Pitch=float * float(A=@n14, B=10))) -> (Output_Get:Rotator)
   > n15 = RandomRotatorFunction!(
    execute
    GameTimeInSeconds:double=GetGameTimeInSeconds()
    PerlinScale:double=1
    Frequency:Rotator="1.5,0.75, 0"
    Amplitude:Rotator="50,-60,0"
    Offset:Rotator="0,-25,0"
    TimeOffset:double=0
   ) -> (then, RandomRotator:Rotator)
   > Set ropeRotator_01(ropeRotator_01:Rotator=@n15) -> (Output_Get:Rotator)
   n16 = GetAnimInstance(self=GetOwnerSkeletalMeshComponent())
   Is Valid(exec=@n2.then_10, InputObject=@n16)
   n17 = ComplexSimSpaceFunction DEPRECATED!(
    execute=@n2.then_2
    CharacterPartAnimInstance:Custom Character Part Anim Instance Object Reference=Self-Reference()
    InterpSpeedMax:double=9
    PawnSpeedMax:double=20
    RInterpErrorTolerance:double=1
    DeltaTime:double=0
    DebugDraw:bool=false
    sim_space_rotation_in:Rotator=Get sim_space_rotation() -> (sim_space_rotation:Rotator)
    inComponentSpaceIn:bool=Get InComponentSpace() -> (InComponentSpace:bool)
    inComponentSpaceLastFrameIn:bool=Get InComponentSpaceLastFrame() -> (InComponentSpaceLastFrame:bool)
    lockedOnSimSpaceIn:bool=Get LockedOnSimSpace() -> (LockedOnSimSpace:bool)
    InterpSpeedIn:double=Get InterpSpeed() -> (InterpSpeed:double)
   ) -> (
    then
    sim_space_rotation:Rotator
    inComponentSpaceOut:bool
    inComponentSpaceLastFrameOut:bool
    lockedOnSimSpaceOut:bool
    InterpSpeedOut:double
   )
   > Set sim_space_rotation(sim_space_rotation:Rotator=@n17) -> (Output_Get:Rotator)
   > Set InComponentSpace(InComponentSpace:bool=@n17.inComponentSpaceOut) -> (Output_Get:bool)
   > Set InComponentSpaceLastFrame(InComponentSpaceLastFrame:bool=@n17.inComponentSpaceLastFrameOut) -> (Output_Get:bool)
   > Set LockedOnSimSpace(LockedOnSimSpace:bool=@n17.lockedOnSimSpaceOut) -> (Output_Get:bool)
   > Set InterpSpeed(InterpSpeed:double=@n17.InterpSpeedOut) -> (Output_Get:double)
   n18 = Gravity Override State Function DEPRECATED(
    execute=@n2.then_3
    CharacterPartAnimInstance=Self-Reference()
    GameTimeInSeconds=GetGameTimeInSeconds()
    GravityOverrideParams=Get Gravity Override Params()
   )
   > Set GravityOverride(GravityOverride=@n18)
   > n19 = Gravity Override State Function DEPRECATED(
    CharacterPartAnimInstance=Self-Reference()
    GameTimeInSeconds=GetGameTimeInSeconds()
    GravityOverrideParams=Get Gravity Override Params_rope() -> (Gravity Override Params_rope:Gravity Override Params Struct Structure)
   )
   > Set GravityOverride_ropes(GravityOverride_ropes:Vector=@n19) -> (Output_Get:Vector)
   > Branch(Condition=CheckForBackbling().bIsBackblingACape)
   > Set GravityOverride_ropes(GravityOverride_ropes="0, 0,-980")
   n20 = PlayerMovement(execute=@n2.then_4, AnimInstance=Self-Reference(), XDotProductCoefficient=1, YDotProductCoefficient=1, ZDotProductCoefficient=1)
   n21 = MapRangeClamped(Value=@n20.YMovement, InRangeA=0, InRangeB=-1)
   n22 = float * float(A="Absolute (Float)"(A=@n20), B=0.5)
   > Set BwdMovement(BwdMovement:double=MapRangeClamped(
     Value=float + float(A=@n22, B=@n21) {is=title K2Node_CommutativeAssociativeBinaryOperator}
     InRangeA=0
     InRangeB=1.1
     OutRangeA=0.4
     OutRangeB=0
    )) -> (Output_Get:double)
   n23 = MapRangeClamped(Value=@n20, InRangeA=-0.2, InRangeB=1)
   n24 = float * float(A="Absolute (Float)"(A=@n20.YMovement), B=0.2)
   > Set RightMovement(RightMovement:double=MapRangeClamped(Value=float + float(A=@n24, B=@n23), InRangeA=0, InRangeB=1.2, OutRangeB=0.5)) -> (Output_Get:double)
   n25 = MapRangeClamped(Value=@n20, InRangeA=0.2, InRangeB=-1, OutRangeB=0.3)
   n26 = float * float(A="Absolute (Float)"(A=@n20.YMovement), B=0.2)
   > Set LeftMovement(LeftMovement:double=MapRangeClamped(Value=float + float(A=@n26, B=@n25), InRangeA=0, InRangeB=1.2, OutRangeB=0.5)) -> (Output_Get:double)
   n27 = Branch(execute=@n2.then_5, Condition=Get bIsSkyDiving())
   > n28 = Branch(Condition="float > float"(A=Get PawnVelocityZ(), B=50) {is=title K2Node_CallFunction})
   > Set BwdMovement(BwdMovement=float * float(A=Get BwdMovement() -> (BwdMovement:double), B=0))
   Set BwdMovement(execute=[@n27.else, @n28.else], BwdMovement=float * float(A=Get BwdMovement(), B=1))
   n29 = GetOwnerSkeletalMeshComponent()
   n30 = BreakRotator(InRot=BreakTransform(InTransform=GetDeltaTransformFromRefPose(self=@n29, BoneName=thigh_r, BaseName=pelvis)).Rotation)
   Set thigh_rot_up_r(execute=@n2.then_6, thigh_rot_up_r:double=MapRangeClamped(Value=@n30.Yaw, InRangeA=-15, InRangeB=-100)) -> (Output_Get:double)
   > Set thigh_rot_inr_r(thigh_rot_inr_r:double=MapRangeClamped(Value=@n30.Pitch, InRangeA=0, InRangeB=-10, OutRangeA=1, OutRangeB=0)) -> (Output_Get:double)
   > Set thigh_rot_otr_r(thigh_rot_otr_r:double=MapRangeClamped(Value=@n30.Pitch, InRangeA=20, InRangeB=50)) -> (Output_Get:double)
   n31 = GetOwnerSkeletalMeshComponent()
   n32 = BreakRotator(InRot=BreakTransform(InTransform=GetDeltaTransformFromRefPose(self=@n31, BoneName=thigh_l, BaseName=pelvis)).Rotation)
   > Set thigh_rot_up_l(thigh_rot_up_l:double=MapRangeClamped(Value=@n32.Yaw, InRangeA=-10, InRangeB=-100)) -> (Output_Get:double)
   > Set thigh_rot_inr_l(thigh_rot_inr_l:double=MapRangeClamped(Value=@n32.Pitch, InRangeA=0, InRangeB=-10, OutRangeA=1, OutRangeB=0)) -> (Output_Get:double)
   > Set thigh_rot_otr_l(thigh_rot_otr_l:double=MapRangeClamped(Value=@n32.Pitch, InRangeA=20, InRangeB=50)) -> (Output_Get:double)
   > Set jumpingAlpha(jumpingAlpha:double="To Float (Boolean)"(InBool=Get bIsJumping())) -> (Output_Get:double)
   n33 = Get bIsMoving2D()
   > Set isNotCrouchRunning(isNotCrouchRunning:double="To Float (Boolean)"(InBool=NOT Boolean(A=@n33))) -> (Output_Get:double)
   > Set ParachuteAlpha(ParachuteAlpha:double="To Float (Boolean)"(InBool=Get bIsParachuteOpened())) -> (Output_Get:double)
   > Set skydiveAlpha(skydiveAlpha:double="To Float (Boolean)"(InBool=Get bIsSkyDiving())) -> (Output_Get:double)
   n34 = Get bIsMoving2D()
   > n35 = Branch(Condition=AND Boolean(A=Get bIsCrouching(), B=NOT Boolean(A=@n34)))
   > Set ropecrouch(ropecrouch:bool=true) -> (Output_Get:bool)
   Set ropecrouch(execute=@n35.else, ropecrouch=false)
   Branch(execute=@n2.then_7, Condition=OR Boolean!(A=Get bIsSprinting(), B=Get bIsSkyDiving()) -> (ReturnValue))
   > Set Head Rot_2(Head Rot_2=0)
   Branch(execute=@n2.then_8, Condition=Get bIsDBNOCarried())
   > Set GravityOverride(GravityOverride="0, 0,-980")
   n36 = GetAnimInstance(self=GetOwnerSkeletalMeshComponent())
   Is Valid(exec=@n2.then_9, InputObject=@n36)
   > Set dance moves(dance moves:bool=Montage_IsPlaying(self=@n36, Montage=/Game/Animation/Game/MainPlayer/Montages/Emotes/Emote_DanceMoves)) -> (Output_Get:bool)
   > Set thrive emote(thrive emote:bool=Montage_IsPlaying(self=@n36, Montage=/BRCosmetics/Animation/Game/MainPlayer/Emotes/Thrive/Emote_Thrive_CMF_M)) -> (Output_Get:bool)
   Event BlueprintInitializeAnimation()
   > Set Skel_Mesh(Skel_Mesh=GetOwnerSkeletalMeshComponent())
   > Set InterpSpeed(InterpSpeed=Get InterpSpeedMax() -> (InterpSpeedMax:double))
   TryGetPawnOwner()
   "Absolute (Float)"(A=Get AimPitch())
   Montage_IsActive!(
    self:Anim Instance Object Reference=@n16
    Montage:Anim Montage Object Reference=/BRCosmetics/Animation/Game/MainPlayer/Emotes/Thrive/Emote_Thrive_CMM_M
   ) -> (ReturnValue:bool)
   GetCurrentActiveMontage!(self:Anim Instance Object Reference) -> (ReturnValue:Anim Montage Object Reference)
   Montage_IsPlaying(self=@n16, Montage=&Emote_Thrive_CMM_M)
   Is Valid(InputObject=GetAnimInstance(self=GetOwnerSkeletalMeshComponent()))
   Get bIsCrouchSprinting()
   Set Head Rot_2(Head Rot_2=1)
   Set GravityOverride(GravityOverride=@n19)
  ]}
]
variables [
 arm l Rot {type=number}
 arm r Rot {type=number}
 Belt Rotator {type=Rotator}
 Belt Rotator_2 {type=Rotator}
 BwdMovement {type=number}
 crown movement {type=Vector}
 dance moves {type=boolean}
 Gravity Override Params {type=GravityOverrideParamsStruct}
 Gravity Override Params_0 {type=GravityOverrideParamsStruct}
 Gravity Override Params_rope {
  type=GravityOverrideParamsStruct
  default {
   emoteOrMelee {gravityOverride=<0 0 -15>, windAmplitude=<50 -200 35>, windFrequency=<0.15 0.7 0.3>, windOffset=<0 -230 -50>}
   front_End {gravityOverride=<0 0 -15>, windAmplitude=<50 -200 35>, windFrequency=<0.15 0.7 0.3>, windOffset=<0 -230 -50>}
   onGround_Moving {gravityOverride=<15 0 -5>, windAmplitude=<50 -200 35>, windFrequency=<0.15 0.7 0.3>, windOffset=<50 -230 -50>}
   onGround_Standing {gravityOverride=<15 5 -25>, windAmplitude=<20 8 50>, windFrequency=<0.5 20 2.5>, windOffset=<15 15 -15>}
   skydive_DivingDown {gravityOverride=<0 0 0>}
   skydive_FloatingDown {gravityOverride=<0 0 0>}
   skydive_GoingUp {gravityOverride=<0 0 -100>}
   skydive_Parachuting {gravityOverride=<0 0 0>}
  }
 }
 GravityOverride {type=Vector}
 GravityOverride_ropes {type=Vector}
 Head Rot_2 {type=number}
 InComponentSpace {type=boolean}
 InComponentSpaceLastFrame {type=boolean}
 InterpSpeed {type=number}
 InterpSpeedMax {type=number}
 isNotCrouchRunning {type=number}
 jumpingAlpha {type=number}
 LeftMovement {type=number}
 LockedOnSimSpace {type=boolean}
 Necklace 01 {type=number}
 ParachuteAlpha {type=number}
 pistol {type=boolean}
 rifle {type=boolean}
 RightMovement {type=number}
 ropecrouch {type=boolean}
 ropeRotator_01 {type=Rotator}
 sim_space_rotation {type=Rotator}
 sineRotation3 {type=Rotator}
 sineRotation4 {type=Rotator}
 sineRotation5 {type=Rotator}
 Skel_Mesh {type=SkeletalMeshComponent}
 skirt fixer {type=boolean}
 skydiveAlpha {type=number}
 slope_foot_rot {type=Rotator}
 slope_root_trans {type=Vector}
 SlopeInfoInput_0 {type=FortAnimInput_SlopeInfo, default {}}
 thigh_rot_inr_l {type=number}
 thigh_rot_inr_r {type=number}
 thigh_rot_l {type=number}
 thigh_rot_l_back {type=number}
 thigh_rot_otr_l {type=number}
 thigh_rot_otr_r {type=number}
 thigh_rot_r {type=number}
 thigh_rot_r_back {type=number}
 thigh_rot_up_l {type=number}
 thigh_rot_up_r {type=number}
 thrive emote {type=boolean}
]
