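# Control Rig for the Thrive Spirit female medium face accessory mesh that procedurally animates three dynamic hair strands.
# The forward solve first applies fixed rotation offsets to the hair C and A root bones.
# It then chains per-item Chain Harmonics on strands B, C and A, blending pendulum lag with differently tuned Y-axis waves, C's running at negative frequency.
# Main hair, front hair and earring bones are left undriven, and harmonics debug drawing is enabled.
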ControlRigBlueprint /BRCosmetics/Characters/Player/Female/Medium/Bodies/F_MED_Thrive_Spirit/Meshes/Parts/F_MED_Thrive_Spirit_FaceAcc_CtrlRig
graphs [
 RigVMFunctionLibrary {links [], nodes []}
 RigVMModel {nodes [
   RigUnit_BeginExecution = &RigUnit_BeginExecution()
   &RigUnit_ModifyTransforms(
    ItemToModify=((Item=(Type=Bone,Name="dyn_hair_c_01"),Transform=(Rotation=(X=0,Y=0,Z=0,W=1),Translation=(X=0,Y=0,Z=0),Scale3D=(X=1,Y=1,Z=1))))
    Weight=1
    ItemToModify.0.Transform.Rotation=&RigVMFunction_MathQuaternionFromEuler(Euler=(X=0,Y=-12,Z=-5)).Result
    ExecutePin=@RigUnit_BeginExecution.ExecutePin
   )
   > &RigUnit_ModifyTransforms(
    ItemToModify=((Item=(Type=Bone,Name="dyn_hair_A_01"),Transform=(Rotation=(X=0,Y=0,Z=0,W=1),Translation=(X=0,Y=0,Z=0),Scale3D=(X=1,Y=1,Z=1))))
    Weight=1
    ItemToModify.0.Transform.Rotation=&RigVMFunction_MathQuaternionFromEuler(Euler=(X=-5,Y=0,Z=0)).Result
   )
   > /Script/ControlRig.RigUnit_ChainHarmonicsPerItem(
    ChainRoot=(Type=Bone,Name="dyn_hair_b_01")
    DrawWorldOffset=(Rotation=(X=0,Y=0,Z=0,W=1),Translation=(X=0,Y=0,Z=0),Scale3D=(X=1,Y=1,Z=1))
    Pendulum=(bEnabled=True,PendulumStiffness=2,PendulumGravity=(X=0,Y=0,Z=0),PendulumBlend=0.75,PendulumDrag=0.98,PendulumMaximum=0.1,UnwindAxis=(X=0,Y=1,Z=0),UnwindMinimum=0.2,UnwindMaximum=0.05)
    Reach=(bEnabled=True,ReachTarget=(X=0,Y=0,Z=0),ReachAxis=(X=1,Y=0,Z=0))
    Speed=(X=1,Y=1,Z=1)
    Wave=(bEnabled=True,WaveFrequency=(X=0,Y=5,Z=0),WaveAmplitude=(X=0,Y=1,Z=0),WaveOffset=(X=0,Y=-5,Z=0),WaveNoise=(X=0,Y=0,Z=0),WaveMaximum=1.25)
    WaveCurve=(EditorCurveData=(Keys=((),(Time=1,Value=1)),DefaultValue=340282346638528859811704183484516925440,PreInfinityExtrap=RCCE_Constant,PostInfinityExtrap=RCCE_Constant))
    bDrawDebug=True
   ) {is=struct RigVMUnitNode}
   > &RigUnit_ChainHarmonicsPerItem(
    ChainRoot=(Type=Bone,Name="dyn_hair_c_01")
    DrawWorldOffset=(Rotation=(X=0,Y=0,Z=0,W=1),Translation=(X=0,Y=0,Z=0),Scale3D=(X=1,Y=1,Z=1))
    Pendulum=(bEnabled=True,PendulumStiffness=2,PendulumGravity=(X=0,Y=0,Z=0),PendulumBlend=0.75,PendulumDrag=0.98,PendulumMaximum=0.1,UnwindAxis=(X=0,Y=1,Z=0),UnwindMinimum=0.2,UnwindMaximum=0.05)
    Reach=(bEnabled=True,ReachTarget=(X=0,Y=0,Z=0),ReachAxis=(X=1,Y=0,Z=0))
    Speed=(X=1,Y=1,Z=1)
    Wave=(bEnabled=True,WaveFrequency=(X=0,Y=-3,Z=0),WaveAmplitude=(X=0,Y=0.5,Z=0),WaveOffset=(X=0,Y=-5,Z=0),WaveNoise=(X=0,Y=0,Z=0),WaveMaximum=1)
    WaveCurve=(EditorCurveData=(Keys=((),(Time=1,Value=1)),DefaultValue=340282346638528859811704183484516925440,PreInfinityExtrap=RCCE_Constant,PostInfinityExtrap=RCCE_Constant))
    bDrawDebug=True
   )
   > &RigUnit_ChainHarmonicsPerItem(
    ChainRoot=(Type=Bone,Name="dyn_hair_A_01")
    DrawWorldOffset=(Rotation=(X=0,Y=0,Z=0,W=1),Translation=(X=0,Y=0,Z=0),Scale3D=(X=1,Y=1,Z=1))
    Pendulum=(bEnabled=True,PendulumStiffness=2,PendulumGravity=(X=0,Y=0,Z=0),PendulumBlend=0.75,PendulumDrag=0.98,PendulumMaximum=0.1,UnwindAxis=(X=0,Y=1,Z=0),UnwindMinimum=0.2,UnwindMaximum=0.05)
    Reach=(bEnabled=True,ReachTarget=(X=0,Y=0,Z=0),ReachAxis=(X=1,Y=0,Z=0))
    Speed=(X=1,Y=1,Z=1)
    Wave=(bEnabled=True,WaveFrequency=(X=0,Y=5,Z=0),WaveAmplitude=(X=0,Y=0.75,Z=0),WaveOffset=(X=0,Y=-5,Z=0),WaveNoise=(X=0,Y=0,Z=0),WaveMaximum=0.8)
    WaveCurve=(EditorCurveData=(Keys=((),(Time=1,Value=1)),DefaultValue=340282346638528859811704183484516925440,PreInfinityExtrap=RCCE_Constant,PostInfinityExtrap=RCCE_Constant))
    bDrawDebug=True
   )
  ]}
]
hierarchy [
 root {loc=<0 0 0>, rot=<0 0 0 1>}
 pelvis {parent=root, loc=<0 -0.5301 98.0571>, rot=<0.0127 0.707 -0.0127 -0.707>}
 spine_01 {parent=pelvis, loc=<6.1208 -0.9328 0>, rot=<0 0 0.034 -0.9994>}
 spine_02 {parent=spine_01, loc=<6.5996 -0.2095 0>, rot=<0 0 0.0238 -0.9997>}
 spine_03 {parent=spine_02, loc=<6.4914 -0.1237 0>, rot=<0 0 0.0149 -0.9999>}
 spine_04 {parent=spine_03, loc=<6.9598 0.0794 0>, rot=<0 0 -0.0132 -0.9999>}
 spine_05 {parent=spine_04, loc=<6.9987 0 0>, rot=<0 0 -0.0015 -1>}
 clavicle_l {parent=spine_05, loc=<5.4831 -0.1053 -3.2946>, rot=<0.0465 -0.7289 -0.009 -0.6829>}
 upperarm_l {parent=clavicle_l, loc=<9.8658 -1.2136 -0.4105>, rot=<0.0164 -0.3245 -0.0269 -0.9454>}
 deform_pectoral_fr_l {parent=clavicle_l, loc=<6.2796 0 0>, rot=<0 0 0 1>}
 clavicle_r {parent=spine_05, loc=<5.4831 -0.1053 3.2946>, rot=<0.7289 0.0465 -0.6829 0.009>}
 upperarm_r {parent=clavicle_r, loc=<-9.8658 1.2136 0.4104>, rot=<0.0164 -0.3245 -0.0269 -0.9454>}
 deform_pectoral_fr_r {parent=clavicle_r, loc=<-6.28 0 0>, rot=<0 0 0 1>}
 neck_01 {parent=spine_05, loc=<11.3138 0.1529 0>, rot=<0 0 -0.2003 -0.9797>}
 neck_02 {parent=neck_01, loc=<4.5627 0 0>, rot=<0 0 0.0504 -0.9987>}
 head {parent=neck_02, loc=<5.3914 0 0>, rot=<0 0 0.115 -0.9934>}
 dyn_hair_main_Look_AT {parent=head, loc=<3.2284 -6.803 -2.9484>, rot=<-0.0235 -0.9993 -0.0276 0.0088>}
 dyn_hair_main_MSTR {parent=dyn_hair_main_Look_AT, loc=<0 0 0>, rot=<0.6633 -0.1774 0.1805 -0.7042>}
 dyn_hair_main_01 {parent=dyn_hair_main_MSTR, loc=<0 0 0>, rot=<0 0.0475 0 -0.9989>}
 dyn_hair_main_02 {parent=dyn_hair_main_01, loc=<6.7062 0 0>, rot=<0 -0.0126 0 -0.9999>}
 dyn_hair_main_03 {parent=dyn_hair_main_02, loc=<7.4378 0 0>, rot=<0 0.0755 0 -0.9971>}
 dyn_hair_main_04 {parent=dyn_hair_main_03, loc=<7.3461 0 0>, rot=<0 0.0635 0 -0.998>}
 dyn_earring_l_MSTR {parent=head, loc=<-0.2078 0.1503 -6.8633>, rot=<0.0048 -0.9817 -0.0001 -0.1904>}
 dyn_earring_l {parent=dyn_earring_l_MSTR, loc=<0 0 0>, rot=<0 0 0 1>}
 dyn_earring_r_MSTR {parent=head, loc=<-0.2078 0.1503 6.8633>, rot=<0.007 -0.9809 0.0038 0.1942>}
 dyn_earring_r {parent=dyn_earring_r_MSTR, loc=<0 0 0>, rot=<0 0 0 1>}
 dyn_front_hair_MSTR {parent=head, loc=<1.7545 -0.7643 -11.0018>, rot=<0.0037 -0.8945 -0.4471 0.0018>}
 dyn_front_hair_01 {parent=dyn_front_hair_MSTR, loc=<0.4518 0 0>, rot=<0 0 0 1>}
 dyn_front_hair_02 {parent=dyn_front_hair_01, loc=<4.7759 4.2586 0.4126>, rot=<0.0665 0 0 -0.9978>}
 dyn_hair_A_MSTR {parent=head, loc=<10.8898 -1.0145 -10.6947>, rot=<0.0029 -0.7071 -0.0029 -0.7071>}
 dyn_hair_A_01 {parent=dyn_hair_A_MSTR, loc=<0.9346 -0.3673 -0.573>, rot=<0.5127 -0.487 -0.5127 -0.487>}
 dyn_hair_A_02 {parent=dyn_hair_A_01, loc=<3.5976 -2.9048 -1.3021>, rot=<0 0 0 1>}
 dyn_hair_A_03 {parent=dyn_hair_A_02, loc=<3.8423 0.4495 -2.5854>, rot=<0 0 0 1>}
 dyn_hair_b_01 {parent=head, loc=<13.8525 -0.1283 -4.5608>, rot=<0.3471 -0.6681 0.1654 -0.637>}
 dyn_hair_b_02 {parent=dyn_hair_b_01, loc=<5.7155 0 0>, rot=<0.0497 -0.3413 -0.0493 -0.9373>}
 dyn_hair_b_03 {parent=dyn_hair_b_02, loc=<4.4147 -0.0051 0.2783>, rot=<0.0247 -0.2571 -0.0123 -0.966>}
 dyn_hair_b_04 {parent=dyn_hair_b_03, loc=<4.0167 0 -0.0151>, rot=<0 -0.192 0 -0.9814>}
 dyn_hair_c_01 {parent=head, loc=<16.3742 -2.7854 -1.1432>, rot=<0.3736 -0.632 0.3664 -0.5717>}
 dyn_hair_c_02 {parent=dyn_hair_c_01, loc=<4.3695 0 0>, rot=<0.0155 -0.2351 -0.0894 -0.9677>}
 dyn_hair_c_03 {parent=dyn_hair_c_02, loc=<5.4507 0 0>, rot=<0.0102 0.0018 -0.2606 -0.9654>}
]
