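# Physics asset for the skydive variant of the Thrive Spirit female medium face accessory, covering a dynamic hair chain and the upper body around it.
# Capsule bodies cover upper spine, clavicles, neck and head, plus a small master hair body and four hair segments.
# Body joints and the head-to-master link swing and twist freely, while the four chained hair joints have narrow angular limits, wider on the second swing axis.
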PhysicsAsset /BRCosmetics/Characters/Player/Female/Medium/Bodies/F_MED_Thrive_Spirit/Meshes/Parts/F_MED_Thrive_Spirit_FaceAcc_skydive_Physics
bodies {
 clavicle_l [{shapeType=Capsule, length=12.6717, radius=4.8754, shapeName=clavicle_l_capsule, center=<5 0 0>, rotation=<rot 90 0 0>}]
 clavicle_r [{shapeType=Capsule, length=13.1634, radius=4.8754, shapeName=clavicle_r_capsule, center=<-5 0 0>, rotation=<rot 90 0 0>}]
 dyn_hair_main_01 [{shapeType=Capsule, length=12.7569, radius=5.7, shapeName=dyn_hair_main_01_capsule, center=<5 0.0001 -2.5>, rotation=<rot 0 0 89.9999>}]
 dyn_hair_main_02 [{shapeType=Capsule, length=10.6463, radius=5.9481, shapeName=dyn_hair_main_02_capsule, center=<5 0 -1.5>, rotation=<rot 0 0 90>}]
 dyn_hair_main_03 [{shapeType=Capsule, length=11.2694, radius=5.3, shapeName=dyn_hair_main_03_capsule, center=<5 0 -1.5>, rotation=<rot 0 0 89.9999>}]
 dyn_hair_main_04 [{shapeType=Capsule, length=3.861, radius=4.8556, shapeName=dyn_hair_main_04_capsule, center=<5 0 -1>, rotation=<rot 0 0 89.9999>}]
 dyn_hair_main_MSTR [{shapeType=Capsule, length=0.505, radius=0.505, shapeName=dyn_hair_main_MSTR_capsule, rotation=<rot 0 -90 0>}]
 head [{shapeType=Capsule, length=8.4367, radius=7.2805, shapeName=head_capsule, center=<-0.9804 2.5998 0>, rotation=<rot 90 90 72.5957>}]
 neck_01 [{shapeType=Capsule, length=4.255, radius=4.9, shapeName=neck_01_capsule, center=<-0.3384 0.7494 0>, rotation=<rot 0 -89.9999 0>}]
 spine_04 [{shapeType=Capsule, length=7.0044, radius=10.6649, shapeName=spine_04_capsule, center=<2.5931 1.6603 0>, rotation=<rot 0 -92.6845 0>}]
 spine_05 [{shapeType=Capsule, length=17.0459, radius=8.0403, shapeName=spine_05_capsule, center=<1.675 -0.1132 -0.155>}]
}
constraints [
 {
  bone1Name=spine_05
  bone2Name=spine_04
  swing1LimitDegrees=45
  swing1Motion=Free
  swing2LimitDegrees=45
  swing2Motion=Free
  twistLimitDegrees=45
  twistMotion=Free
 }
 {
  bone1Name=clavicle_l
  bone2Name=spine_05
  swing1LimitDegrees=45
  swing1Motion=Free
  swing2LimitDegrees=45
  swing2Motion=Free
  twistLimitDegrees=45
  twistMotion=Free
 }
 {
  bone1Name=clavicle_r
  bone2Name=spine_05
  swing1LimitDegrees=45
  swing1Motion=Free
  swing2LimitDegrees=45
  swing2Motion=Free
  twistLimitDegrees=45
  twistMotion=Free
 }
 {
  bone1Name=head
  bone2Name=neck_01
  swing1LimitDegrees=45
  swing1Motion=Free
  swing2LimitDegrees=45
  swing2Motion=Free
  twistLimitDegrees=45
  twistMotion=Free
 }
 {bone1Name=dyn_hair_main_01, bone2Name=dyn_hair_main_MSTR, swing1LimitDegrees=15, swing2LimitDegrees=35, twistLimitDegrees=15}
 {bone1Name=dyn_hair_main_02, bone2Name=dyn_hair_main_01, swing1LimitDegrees=15, swing2LimitDegrees=35, twistLimitDegrees=15}
 {bone1Name=dyn_hair_main_03, bone2Name=dyn_hair_main_02, swing1LimitDegrees=15, swing2LimitDegrees=35, twistLimitDegrees=15}
 {bone1Name=dyn_hair_main_04, bone2Name=dyn_hair_main_03, swing1LimitDegrees=15, swing2LimitDegrees=35, twistLimitDegrees=15}
 {
  bone1Name=dyn_hair_main_MSTR
  bone2Name=head
  swing1LimitDegrees=45
  swing1Motion=Free
  swing2LimitDegrees=45
  swing2Motion=Free
  twistLimitDegrees=45
  twistMotion=Free
 }
 {
  bone1Name=neck_01
  bone2Name=spine_05
  swing1LimitDegrees=45
  swing1Motion=Free
  swing2LimitDegrees=45
  swing2Motion=Free
  twistLimitDegrees=45
  twistMotion=Free
 }
]
